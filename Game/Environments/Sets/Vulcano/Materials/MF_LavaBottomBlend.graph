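# Material function for Vulcano lava pile-ups, outputting full material attributes plus a separate WPO.
# Local-space XY becomes panning radial UVs for mask and normal textures, sampled with mesh-UV derivatives, likely avoiding polar seams.
# Red vertex color and mask drive emissive PileColor and depth-faded opacity, multiplied by particle alpha.
# A blurred-mip mask adds Z displacement to a vertex-red offset on all axes, scaled by object Z scale.
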MaterialFunction /Game/Environments/Sets/Vulcano/Materials/MF_LavaBottomBlend
graph {nodes [
  ComponentMask() {g=true, r=true}
  n1 = VectorParameter() {parameterName=PileColor, defaultValue=<rgba 9 1.125 0 1>}
  n2 = VertexColor()
  n3 = FunctionInput(Preview=WorldPosition() {worldPositionShaderOffset=WPT_ExcludeAllShaderOffsets}.XYZ) {bUsePreviewValueAsDefault=true, inputName=Default World Position}
  n4 = Divide(
   A=ComponentMask(None=TransformPosition(None=@n3) {transformSourceType=TRANSFORMPOSSOURCE_World}) {g=true, r=true}
   B=AppendVector(
    A=ScalarParameter() {defaultValue=444.3143, parameterName=RadialSize2}
    B=ScalarParameter() {defaultValue=464.9501, parameterName=RadialSizeY2}
   )
  )
  n5 = ScalarParameter() {defaultValue=2, parameterName=RadialTilingX2}
  n6 = ScalarParameter() {defaultValue=-3, parameterName=RadialTilingY2}
  n7 = Time()
  n8 = AppendVector(A=ScalarParameter() {defaultValue=0.05, parameterName=Speed_x2}, B=ScalarParameter() {defaultValue=0.1429, parameterName=Speed_y2})
  n9 = Add(
   A=Multiply(
    A=MaterialFunctionCall(Vector or UVs=@n4) {materialFunction=&VectorToRadialValue}.Radial Coordinates
    B=AppendVector(A=@n5, B=@n6)
   )
   B=Frac(None=Multiply(A=@n7, B=@n8))
  )
  n10 = TextureCoordinate()
  n11 = DDX(Value=@n10)
  n12 = DDY(Value=@n10)
  n13 = TextureSample!(UVs=@n9, Tex, "DDX(UVs)"=@n11, "DDY(UVs)"=@n12, Apply View MipBias) {mipValueMode=TMVM_Derivative, samplerType=SAMPLERTYPE_Masks, texture=/Game/Environments/Sets/Vulcano/Textures/LavaPileUp_M}
  n14 = Power(Base=Max(A=@n2.R, B=@n13.R), Exp=ScalarParameter() {defaultValue=1.6371, parameterName=powerRings2})
  n15 = ParticleColor()
  n16 = Add(A=Multiply(A=@n2.R, B=ScalarParameter() {defaultValue=0.3238, parameterName=CenterGlow}), B=@n13.R)
  n17 = VertexColor()
  n18 = DepthFade(
   Opacity=Multiply(A=Clamp(None=@n16), B=@n17.R)
   FadeDistance=LinearInterpolate(A=ScalarParameter() {parameterName=DepthFade}, B=ScalarParameter() {defaultValue=19.6863, parameterName=DepthFade2}, Alpha=@n17.R)
  )
  n19 = Multiply(
   A=TextureSample!(UVs=@n9, Tex, Level=ScalarParameter() {defaultValue=6.6088, parameterName=BlurWPOTexture}, Apply View MipBias) {mipValueMode=TMVM_MipLevel, samplerType=SAMPLERTYPE_Masks, texture=&LavaPileUp_M}.G
   B=Constant3Vector() {constant=<rgba 0 0 1 0>}
  )
  n20 = Multiply(
   A=Add(A=Multiply(A=@n19, B=@n17.R), B=@n17.R)
   B=Multiply(
    A=ScalarParameter() {defaultValue=18.066, parameterName=PileOffsetWPO}
    B=MaterialFunctionCall() {materialFunction=/Engine/Functions/Engine_MaterialFunctions02/WorldPositionOffset/ObjectScale}.Scale Z
   )
  )
  FunctionOutput(None=MakeMaterialAttributes!(
    BaseColor
    Metallic
    Specular
    Roughness
    Anisotropy
    EmissiveColor=LinearInterpolate(B=@n1.RGB, Alpha=@n14)
    Opacity=Multiply(A=@n15.A, B=@n18)
    OpacityMask
    Normal=TextureSample!(UVs=@n9, Tex, "DDX(UVs)"=@n11, "DDY(UVs)"=@n12, Apply View MipBias) {mipValueMode=TMVM_Derivative, samplerType=SAMPLERTYPE_Normal, texture=/Game/Environments/Sets/Vulcano/Textures/LavaPileUp_N}.RGB
    Tangent
    WorldPositionOffset=@n20
    SubsurfaceColor
    ClearCoat
    ClearCoatRoughness
    AmbientOcclusion
    Refraction
    CustomizedUVs_0
    CustomizedUVs_1
    CustomizedUVs_2
    CustomizedUVs_3
    CustomizedUVs_4
    CustomizedUVs_5
    CustomizedUVs_6
    CustomizedUVs_7
    PixelDepthOffset
    ShadingModel
    Displacement
   ))
  FunctionOutput(None=@n20) {outputName=WPO, sortPriority=2}
 ]}
