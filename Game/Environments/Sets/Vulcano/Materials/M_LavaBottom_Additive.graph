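# Additive unlit material for the lava bottom in the Vulcano environment set, tagged with the Lava physical material.
# All shading comes from the MF_LavaBottomBlend function, whose result passes through MF_CreativeMoveTool into the material attributes output.
# It supports Niagara mesh particles and allows translucent custom depth writes.
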Material /Game/Environments/Sets/Vulcano/Materials/M_LavaBottom_Additive
graph {
 properties {
  allowTranslucentCustomDepthWrites=true
  bUseMaterialAttributes=true
  bUsedWithNiagaraMeshParticles=true
  blendMode=BLEND_Additive
  physMaterial=/Game/FortressPhysicalMaterials/Lava
  pixelDepthOffsetMode=PDOM_Legacy
  shadingModel=MSM_Unlit
  usedShadingModels=Unlit
 }
 nodes [MaterialFunctionCall(MA In=MaterialFunctionCall!(Default World Position) {materialFunction=&/MF_LavaBottomBlend}.Result) {materialFunction=&MF_CreativeMoveTool}]
}
