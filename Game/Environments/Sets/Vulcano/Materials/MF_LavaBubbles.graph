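# Material function that adds animated, pulsing bubbles to a lava material from a splatter mask and normal texture.
# Bump offset gives the bubbles parallax, and a clamped time sine, phase-shifted by the mask's blue channel, drives a HeightLerp on bubble height.
# The resulting alpha blends the bubble normal into the input attributes and shapes an additive emissive output.
# A third output derives UV distortion from the pulse, red mask and blurred higher-mip normal.
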MaterialFunction /Game/Environments/Sets/Vulcano/Materials/MF_LavaBubbles
graph {nodes [
  n1 = GetMaterialAttributes(None=FunctionInput() {inputType=FunctionInput_MaterialAttributes}) {attributeGetTypes [#1]}
  n2 = ComponentMask(None=FunctionInput() {inputName=UVs, inputType=FunctionInput_Vector2, sortPriority=3}) {desc=UVInput, g=true, r=true}
  n3 = BumpOffset(
   Coordinate=@n2
   Height=TextureSample(UVs=@n2) {samplerType=SAMPLERTYPE_Masks, texture=/Game/Global/Textures/Default/PointSplatter/Bubbles_M}.G
   HeightRatioInput=ScalarParameter() {defaultValue=0.02, parameterName=BubbleHeight}
  ) {referencePlane=0}
  n4 = TextureSample(UVs=@n3) {samplerType=SAMPLERTYPE_Normal, texture=/Game/Global/Textures/Default/PointSplatter/Bubbles_N}
  n5 = Multiply(A=Time(), B=ScalarParameter() {defaultValue=0.1, parameterName=BubbleSpeed})
  n6 = TextureSample(UVs=@n3) {samplerType=SAMPLERTYPE_Masks, texture=&Bubbles_M}
  n7 = Multiply(A=ScalarParameter() {defaultValue=1, parameterName=BubbleTimeOffset}, B=@n6.B)
  n8 = Clamp(None=Sine(None=Add(A=@n5, B=@n7)))
  n9 = MaterialFunctionCall(Transition Phase=@n8, Height Texture=@n6.G, Contrast=ScalarParameter() {defaultValue=1, parameterName=BubbleContrast}) {materialFunction=&HeightLerp}
  n10 = Multiply(A=@n9.Alpha, B=Clamp(None=Multiply(A=@n6.G) {constB=50}))
  FunctionOutput(None=SetMaterialAttributes!(MaterialAttributes=@n1.MaterialAttributes, Normal=LinearInterpolate(A=@n1.Normal, B=@n4.RGB, Alpha=@n10)) {attributeSetTypes [#1]})
  n11 = ScalarParameter() {defaultValue=0.3971, parameterName=BubbleEmissive}
  n12 = Multiply(A=@n8, B=@n9.Lerp Alpha No Contrast)
  FunctionOutput(None=Multiply(A=Multiply(A=@n11, B=@n12), B=@n10)) {outputName=BubblesAdditive, sortPriority=3}
  n13 = ScalarParameter() {defaultValue=6.8854, parameterName=LevelNormalBlur}
  n14 = Multiply(
   A=Multiply(A=@n8, B=@n6.R)
   B=ComponentMask(None=TextureSample!(UVs=@n3, Tex, Level=@n13, Apply View MipBias) {mipValueMode=TMVM_MipLevel, samplerType=SAMPLERTYPE_Normal, texture=&Bubbles_N}.RGB) {g=true, r=true}
  )
  n15 = ScalarParameter() {defaultValue=-0.8337, parameterName=UVDistortion}
  FunctionOutput(None=ComponentMask(None=Multiply(A=@n14, B=@n15)) {desc=UVDistortion, g=true, r=true}) {outputName=BubblesUVDistortion, sortPriority=4}
 ]}
