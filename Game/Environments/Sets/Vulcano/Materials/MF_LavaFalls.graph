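# Material function for Vulcano lava falls, outputting material attributes with emissive color, opacity mask and normal.
# Lava mask and normal textures scroll over time, with tiling and a per-object offset derived from world position.
# The inverted mask combines with UV2 gradients and HeightLerp to shape glow and cutout; IsHLOD skips the first gradient stage.
# Glow color blends between a fixed color and time-of-day colors via UseTODColors.
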MaterialFunction /Game/Environments/Sets/Vulcano/Materials/MF_LavaFalls
graph {nodes [
  n1 = LinearInterpolate(
   A=VectorParameter() {parameterName=FallsColor, defaultValue=<rgba 2 0.25 0 1>}.RGB
   B=MaterialFunctionCall(
    Color_Morning=VectorParameter() {parameterName=StreamColor_Morning, defaultValue=<rgba 5 1.1178 0.4948 1>}.RGB
    Color_Day=VectorParameter() {parameterName=StreamColor_Day, defaultValue=<rgba 5 1.1178 0.4948 1>}.RGB
    Color_Evening=VectorParameter() {parameterName=StreamColor_Evening, defaultValue=<rgba 5 1.1178 0.4948 1>}.RGB
    Color_Night=VectorParameter() {parameterName=StreamColor_Night, defaultValue=<rgba 5 1.1178 0.4948 1>}.RGB
   ) {materialFunction=&Color_Time_of_Day}.Result
   Alpha=ScalarParameter() {parameterName=UseTODColors}
  )
  n2 = TextureCoordinate()
  n3 = ObjectPositionWS()
  n4 = Multiply(A=ComponentMask(None=@n3) {r=true}) {constB=0.33}
  n5 = ComponentMask(None=@n3) {g=true}
  n6 = Frac(None=AppendVector(A=Divide(A=@n4, B=@n5), B=Divide(A=@n5, B=@n4)))
  n7 = ScalarParameter() {defaultValue=1, parameterName=FallsTilingX}
  n8 = ScalarParameter() {defaultValue=2.4505, parameterName=FallsTilingZ}
  n9 = Constant()
  n10 = Multiply(A=Time(), B=ScalarParameter() {defaultValue=-0.1, parameterName=FallsSpeed1})
  n11 = Add(A=Divide(A=Add(A=@n2, B=@n6), B=AppendVector(A=@n7, B=@n8)), B=Frac(None=AppendVector(A=@n9, B=@n10)))
  n12 = OneMinus(None=TextureSample(UVs=@n11) {samplerType=SAMPLERTYPE_Masks, texture=/Game/Environments/Sets/Vulcano/Textures/Lava_M}.G)
  n13 = ScalarParameter() {defaultValue=-6.5449, parameterName=GradienBottom}
  n14 = ScalarParameter() {defaultValue=3.4402, parameterName=GradienBottom2}
  n15 = TextureCoordinate() {coordinateIndex=2}
  n16 = ComponentMask(None=@n15) {g=true}
  n17 = ScalarParameter() {defaultValue=-0.7775, parameterName=FalloffEdges}
  n18 = ScalarParameter() {defaultValue=2.7126, parameterName=FalloffEdges2}
  n19 = ComponentMask(None=@n15) {r=true}
  n20 = Multiply(A=@n19, B=OneMinus(None=@n19))
  n21 = Add(
   A=Clamp(None=LinearInterpolate(A=@n13, B=@n14, Alpha=@n16)) {clampMode=CMODE_ClampMin}
   B=Add(A=LinearInterpolate(A=@n17, B=@n18, Alpha=@n20), B=@n12)
  )
  n22 = ScalarParameter() {defaultValue=6.9247, parameterName=InnerEdge1}
  n23 = ScalarParameter() {defaultValue=-1.689, parameterName=InnerEdge2}
  n24 = Multiply(A=@n16, B=OneMinus(None=@n16))
  n25 = MaterialFunctionCall(
   Transition Phase=Multiply(A=OneMinus(None=@n24), B=ScalarParameter() {defaultValue=0.6605, parameterName=HeightblendFalls})
   Height Texture=@n12
  ) {materialFunction=&HeightLerp}
  n26 = Add(
   A=Power(Base=StaticSwitchParameter(True=@n12, False=@n21) {parameterName=IsHLOD}, Exp=ScalarParameter() {defaultValue=1.1253, parameterName=power})
   B=Clamp(None=LinearInterpolate(A=@n22, B=@n23, Alpha=@n25.Lerp Alpha No Contrast)) {clampMode=CMODE_ClampMin}
  )
  FunctionOutput(None=MakeMaterialAttributes!(
    BaseColor
    Metallic
    Specular
    Roughness
    Anisotropy
    EmissiveColor=LinearInterpolate(B=@n1, Alpha=@n26)
    Opacity
    OpacityMask=@n25.Alpha
    Normal=TextureSample(UVs=@n11) {samplerType=SAMPLERTYPE_Normal, texture=/Game/Environments/Sets/Vulcano/Textures/Lava_N}.RGB
    Tangent
    WorldPositionOffset
    SubsurfaceColor
    ClearCoat
    ClearCoatRoughness
    AmbientOcclusion
    Refraction
    CustomizedUVs_0
    CustomizedUVs_1
    CustomizedUVs_2
    CustomizedUVs_3
    CustomizedUVs_4
    CustomizedUVs_5
    CustomizedUVs_6
    CustomizedUVs_7
    PixelDepthOffset
    ShadingModel
    Displacement
   ))
  Multiply(A=OneMinus(None=@n25.Lerp Alpha No Contrast), B=@n23)
  WorldPosition()
 ]}
