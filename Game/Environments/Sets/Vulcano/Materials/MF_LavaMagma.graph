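# Material function for Vulcano lava, outputting material attributes with glow colour in base colour and emissive, plus UV, large-scale mask and flow vector outputs.
# World-XY UVs with sine wobble and optional distortion sample detail, optionally flow-mapped or distance-faded.
# A large-scale mask perturbs an animated UV1 or vertex-colour gradient lookup, which combines with detail and additive bubbles to blend two glow colours.
# Normal, BlackBody and bubbling-distortion branches are left unconnected.
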MaterialFunction /Game/Environments/Sets/Vulcano/Materials/MF_LavaMagma
graph {nodes [
  n1 = ComponentMask(None=WorldPosition().XYZ) {g=true, r=true}
  n2 = ScalarParameter() {defaultValue=653.4476, parameterName=LavaSizeMaster}
  n3 = Add(A=WorldPosition().XYZ, B=Multiply(A=Time(), B=ScalarParameter() {defaultValue=80, parameterName=WaveySpeed}))
  n4 = ScalarParameter() {defaultValue=224.3151, parameterName=WaveDistance}
  n5 = Sine(None=Frac(None=Divide(A=@n3, B=@n4)))
  n6 = ScalarParameter() {defaultValue=0.015, parameterName=WaveyWPO}
  n7 = Add(
   A=FunctionInput(Preview=Divide(A=@n1, B=@n2)) {bUsePreviewValueAsDefault=true, inputName=UVs, inputType=FunctionInput_Vector2}
   B=ComponentMask(None=Multiply(A=@n5, B=@n6)) {g=true, r=true}
  )
  n8 = Add(
   A=@n7
   B=FunctionInput(Preview=Constant()) {bUsePreviewValueAsDefault=true, inputName=UVDistortion, inputType=FunctionInput_Vector2, sortPriority=5}
  )
  n9 = ComponentMask(None=TextureSample!(UVs=@n8, Tex, Level=ScalarParameter() {defaultValue=1, parameterName=BlurMip}, Apply View MipBias) {mipValueMode=TMVM_MipLevel, samplerType=SAMPLERTYPE_Normal, texture=/Game/Environments/Sets/Vulcano/Textures/Lava_N}.RGB) {g=true, r=true}
  n10 = Time()
  n11 = ScalarParameter() {parameterName=SpeedBubbling}
  n12 = Panner(
   Coordinate=Multiply(A=@n7, B=ScalarParameter() {defaultValue=0.9344, parameterName=TilingMask})
   Speed=AppendVector(A=ScalarParameter() {parameterName=SpeedX2}, B=ScalarParameter() {defaultValue=0.01, parameterName=SpeedY2})
  )
  n13 = Sine(None=Add(
    A=Multiply(A=@n10, B=@n11)
    B=TextureSample!(UVs=@n12, Tex, Apply View MipBias) {samplerType=SAMPLERTYPE_Masks, texture=/Game/Environments/Sets/Vulcano/Textures/Lava_M}.G
   ))
  Add(A=Multiply(A=ScalarParameter() {parameterName=DistortionGrow}, B=Multiply(A=@n9, B=@n13)), B=@n8)
  n14 = TextureSampleParameter2D(UVs=TextureCoordinate() {coordinateIndex=4}) {parameterName=FlowMap, samplerType=SAMPLERTYPE_Normal, texture=/Game/Environments/Sets/Vulcano/Textures/Lava_Exterior_FlowMap_Flow}
  n15 = Multiply(
   A=MaterialFunctionCall(XY=ComponentMask(None=@n14.RGB) {g=true, r=true}) {materialFunction=&Swizzle}.YX
   B=Constant2Vector() {g=-1, r=-1}
  )
  n16 = MaterialFunctionCall!(
   Diffuse=TextureObject() {samplerType=SAMPLERTYPE_Masks, texture=&Lava_M}
   Normal Map=TextureObject() {samplerType=SAMPLERTYPE_Normal, texture=&Lava_N}
   "Flow Vector Map (see tooltip)"=@n15
   UV texture 2 offset
   "Time (see tooltip)"=Multiply(A=Time(), B=ScalarParameter() {defaultValue=0.2, parameterName=FlowSpeed})
   useMipLevel
   UVs=@n8
   Texture Size
   Texture Mip Bias
  ) {materialFunction=/Engine/Functions/Engine_MaterialFunctions02/Texturing/Flowmaps.FlowMaps}
  Add(A=@n16.UVs)
  Add()
  n17 = Multiply(A=@n7, B=ScalarParameter() {defaultValue=0.0767, parameterName=LargeScalePatternSize2})
  n18 = StaticSwitchParameter(
   True=ComponentMask(None=@n16.Diffuse) {g=true}
   False=TextureSample(UVs=@n8) {samplerType=SAMPLERTYPE_Masks, texture=&Lava_M}.G
  ) {defaultValue=true, parameterName=FlowMap}
  Add(A=@n17, B=Multiply(A=ScalarParameter() {defaultValue=0.009, parameterName=DistortLargeScale2}, B=@n18))
  n19 = Time()
  n20 = ScalarParameter() {defaultValue=-0.0738, parameterName=GradientMotionSpeed}
  n21 = ComponentMask(None=TextureCoordinate() {coordinateIndex=1}) {g=true}
  n22 = Divide(A=Add(A=Multiply(A=@n19, B=@n20), B=@n21), B=ScalarParameter() {defaultValue=0.4381, parameterName=GradientMotionWaveLength})
  n23 = Multiply(A=Sine(None=Frac(None=@n22)), B=ScalarParameter() {defaultValue=0.04, parameterName=GradientMotionDistance})
  n24 = ScalarParameter() {defaultValue=0.005, parameterName=WaveyWPO2}
  n25 = Add(A=@n21, B=ComponentMask(None=Multiply(A=@n5, B=@n24)) {r=true})
  n26 = ScalarParameter() {defaultValue=0.0767, parameterName=LargeScalePatternSize}
  n27 = ScalarParameter() {defaultValue=0.009, parameterName=DistortLargeScale}
  n28 = TextureSample(UVs=Add(A=Multiply(A=@n7, B=@n26), B=Multiply(A=@n27, B=@n18))) {samplerType=SAMPLERTYPE_Masks, texture=&Lava_M}
  n29 = ScalarParameter() {defaultValue=0.0767, parameterName=LargeScalePatternSize2IntensityMin}
  n30 = ScalarParameter() {defaultValue=0.0767, parameterName=LargeScalePatternSize2IntensityMax}
  n31 = TextureSample(UVs=@n17) {samplerType=SAMPLERTYPE_Masks, texture=&Lava_M}
  n32 = StaticSwitchParameter(True=Add(A=@n28.R, B=LinearInterpolate(A=@n29, B=@n30, Alpha=@n31.R)), False=@n28.R) {parameterName=Use3rdLayerMask}
  n33 = Subtract(A=@n32) {constB=0.5}
  n34 = ScalarParameter() {defaultValue=0.542, parameterName=UVGradientTextureContribution}
  n35 = Multiply(A=@n32, B=@n34)
  n36 = StaticSwitchParameter(True=Add(A=@n25, B=Multiply(A=@n33, B=@n34)), False=Multiply(A=@n25, B=OneMinus(None=@n35))) {defaultValue=true, parameterName=UseGradientAdd_or_Mulitply}
  n37 = VertexColor()
  n38 = Add(A=OneMinus(None=@n37.A), B=Multiply(A=@n37.G, B=@n32))
  n39 = TextureSample(UVs=StaticSwitchParameter(True=Add(A=@n23, B=@n36), False=Clamp(None=@n38)) {defaultValue=true, parameterName=UseUVChannel2Gradient}) {samplerType=SAMPLERTYPE_Masks, texture=/Game/Environments/Sets/Vulcano/Textures/Lava_Gradient}
  n40 = LinearInterpolate(
   A=ScalarParameter() {defaultValue=1, parameterName=FadeDetailNear}
   B=ScalarParameter() {defaultValue=1, parameterName=FadeDetailFar}
   Alpha=MaterialFunctionCall(Distance=ScalarParameter() {defaultValue=10000, parameterName=FadeDistanceDetailTexture}) {materialFunction=&Near_Camera_Fade}.Result
  )
  n41 = StaticSwitchParameter(True=Saturate(None=Multiply(A=@n18, B=@n40)), False=@n18) {parameterName=UseDistanceFade}
  n42 = FunctionInput(Preview=Constant()) {bUsePreviewValueAsDefault=true, inputName=BubblesAdditive, inputType=FunctionInput_Scalar, sortPriority=3}
  n43 = Add(A=Multiply(A=OneMinus(None=@n39.R), B=@n41), B=@n42)
  n44 = ScalarParameter() {defaultValue=0.7265, parameterName=LavaContrastBlackBody}
  BlackBody!(Temp=Multiply(A=ScalarParameter() {defaultValue=2764.4399, parameterName=Temperature}, B=Power(Base=@n43, Exp=@n44)))
  Constant2Vector() {g=1, r=1}
  n45 = OneMinus(None=@n39.R)
  n46 = ScalarParameter() {defaultValue=3.5775, parameterName=LavaContrast2}
  n47 = Power(
   Base=Max(A=@n41) {constB=0}
   Exp=LinearInterpolate(
    A=ScalarParameter() {defaultValue=0.6474, parameterName=LavaContrast}
    B=ScalarParameter() {defaultValue=-0.8378, parameterName=LavaContrast3}
    Alpha=Subtract(A=@n32, B=@n39.R)
   )
  )
  n48 = LinearInterpolate(
   A=VectorParameter() {parameterName=LavalGlow_1, defaultValue=<rgba 0 0 0 1>}.RGB
   B=Multiply(
    A=ScalarParameter() {defaultValue=0.5143, parameterName=LavaBrightness}
    B=VectorParameter() {parameterName=LavalGlow_2, defaultValue=<rgba 5 0.5 0 1>}.RGB
   )
   Alpha=Multiply(A=Power(Base=@n45, Exp=@n46), B=Add(A=@n47, B=@n42))
  )
  FunctionOutput(None=MakeMaterialAttributes!(
    BaseColor=@n48
    Metallic
    Specular=ScalarParameter() {defaultValue=0.05, parameterName=Specular}
    Roughness=ScalarParameter() {defaultValue=1, parameterName=Roughness}
    Anisotropy
    EmissiveColor=@n48
    Opacity
    OpacityMask
    Normal
    Tangent
    WorldPositionOffset
    SubsurfaceColor
    ClearCoat
    ClearCoatRoughness
    AmbientOcclusion
    Refraction
    CustomizedUVs_0
    CustomizedUVs_1
    CustomizedUVs_2
    CustomizedUVs_3
    CustomizedUVs_4
    CustomizedUVs_5
    CustomizedUVs_6
    CustomizedUVs_7
    PixelDepthOffset
    ShadingModel
    Displacement
   ))
  FunctionOutput(None=@n7) {outputName=UVs Panning, sortPriority=1}
  FunctionOutput(None=@n32) {outputName=LargeScaleMask, sortPriority=8}
  FunctionOutput(None=@n15) {outputName=FlowMap, sortPriority=5}
  Panner(
   Coordinate=@n7
   Speed=AppendVector(A=ScalarParameter() {defaultValue=0.02, parameterName=SpeedX}, B=ScalarParameter() {defaultValue=0.005, parameterName=SpeedY})
  )
  StaticSwitchParameter(True=@n16.Normal, False=TextureSample(UVs=@n8) {samplerType=SAMPLERTYPE_Normal, texture=&Lava_N}.RGB) {defaultValue=true, parameterName=FlowMap}
  StaticSwitchParameter(True=@n7, False=@n7) {parameterName=IsLavalStream}
  VectorParameter() {parameterName=LavalGlow_3, defaultValue=<rgba 0 0 0 1>}
 ]}
