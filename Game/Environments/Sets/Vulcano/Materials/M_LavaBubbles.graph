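# Two-sided lava master material for the Vulcano environment set, with static switches for bubbles, streams and lava falls.
# MF_LavaMagma and MF_LavaBubbles form the world-XY-mapped surface; streams add panning masks and time-of-day color, while falls blend in MF_LavaFalls by UV3.
# Real bubbles pulse on a position-seeded timer, scaling about their pivot via WPO and driving brightness.
# An optional camera-axis offset fixes z-fighting before MF_CreativeMoveTool.
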Material /Game/Environments/Sets/Vulcano/Materials/M_LavaBubbles
graph {
 properties {
  bTangentSpaceNormal=false
  bUseMaterialAttributes=true
  bUsedWithInstancedStaticMeshes=true
  bUsedWithNanite=true
  bUsedWithSplineMeshes=true
  physMaterial=/Game/FortressPhysicalMaterials/Lava
  pixelDepthOffsetMode=PDOM_Legacy
  twoSided=true
  usedShadingModels=Default Lit
 }
 nodes [
  n1 = Add(A=WorldPosition().XYZ, B=Multiply(A=Time(), B=ScalarParameter() {defaultValue=35.3185, parameterName=WaveySpeed}))
  n2 = ScalarParameter() {defaultValue=71.0708, parameterName=WaveDistance}
  n3 = Sine(None=Frac(None=Divide(A=@n1, B=@n2)))
  n4 = ScalarParameter() {defaultValue=5.2267, parameterName=WaveyWPO}
  Add(B=Multiply(A=Constant3Vector() {constant=<rgba 1 1 0 0>}, B=Multiply(A=@n3, B=@n4)))
  n5 = VertexNormalWS()
  Clamp(None=LinearInterpolate(
    A=ScalarParameter() {defaultValue=3.1864, parameterName=LavaFallTransition1}
    B=ScalarParameter() {defaultValue=0.1905, parameterName=LavaFallTransition2}
    Alpha=ComponentMask(None=@n5) {b=true}
   ))
  n6 = ObjectPositionWS()
  n7 = ComponentMask(None=@n6) {r=true}
  n8 = ComponentMask(None=@n6) {g=true}
  Divide(A=@n7, B=@n8)
  n9 = WorldPosition()
  n10 = Constant()
  n11 = Multiply(
   A=ComponentMask(None=TextureCoordinate() {coordinateIndex=1}) {g=true}
   B=ScalarParameter() {defaultValue=7.6023, parameterName=RealBubbleBrightness}
  )
  n12 = Multiply(A=Frac(None=@n7), B=Frac(None=@n8))
  n13 = Constant() {r=34252}
  n14 = Multiply(A=Add(A=Time(), B=Multiply(A=@n12, B=@n13)), B=ScalarParameter() {defaultValue=0.4089, parameterName=Speed})
  n15 = ScalarParameter() {defaultValue=8, parameterName=PhaseLength}
  n16 = Multiply(A=Frac(None=Divide(A=@n14, B=@n15)), B=@n15)
  n17 = Ceil(None=Frac(None=Clamp(None=@n16)))
  n18 = Multiply(A=@n17, B=Frac(None=@n14))
  n19 = MaterialFunctionCall!(
   UVs=Divide(A=ComponentMask(None=@n9.XYZ) {g=true, r=true}, B=ScalarParameter() {defaultValue=653.4476, parameterName=LavaSizeMaster})
   BubblesAdditive=StaticSwitchParameter(True=Multiply(A=@n11, B=@n18), False=StaticSwitchParameter(True=@n20.BubblesAdditive, False=@n10) {parameterName=UseBubbles}) {parameterName=IsRealBubble}
   UVDistortion=StaticSwitchParameter(True=@n20.BubblesUVDistortion, False=@n10) {parameterName=UseBubbles}
  ) {materialFunction=&/MF_LavaMagma}
  n20 = MaterialFunctionCall!(In=@n19.Result, UVs=@n19.UVs Panning) {materialFunction=&/MF_LavaBubbles}
  n21 = StaticSwitchParameter(True=@n20.Result, False=@n19.Result) {parameterName=UseBubbles}
  n22 = GetMaterialAttributes(None=@n21) {attributeGetTypes [#1]}
  n23 = ScalarParameter() {parameterName=PanSpeedStream2_X}
  n24 = ScalarParameter() {defaultValue=-0.05, parameterName=PanSpeedStream2_Y}
  n25 = TextureSample(UVs=Panner(Coordinate=TextureCoordinate(), Speed=AppendVector(A=@n23, B=@n24))) {samplerType=SAMPLERTYPE_Masks, texture=/Game/Environments/Sets/Vulcano/Textures/LavaStream_M}
  n26 = VertexColor()
  n27 = Multiply(A=Clamp(None=@n25.R), B=@n26.R)
  n28 = VectorParameter() {parameterName=StreamColor_Morning, defaultValue=<rgba 5 1.1178 0.4948 1>}
  n29 = VectorParameter() {parameterName=StreamColor_Day, defaultValue=<rgba 5 1.1178 0.4948 1>}
  n30 = VectorParameter() {parameterName=StreamColor_Evening, defaultValue=<rgba 5 1.1178 0.4948 1>}
  n31 = VectorParameter() {parameterName=StreamColor_Night, defaultValue=<rgba 5 1.1178 0.4948 1>}
  n32 = LinearInterpolate(
   A=@n22.Emissive Color
   B=LinearInterpolate(
    A=VectorParameter() {parameterName=StreamColor_Dark, defaultValue=<rgba 0 0 0 1>}.RGB
    B=MaterialFunctionCall(Color_Morning=@n28.RGB, Color_Day=@n29.RGB, Color_Evening=@n30.RGB, Color_Night=@n31.RGB) {materialFunction=&Color_Time_of_Day}.Result
    Alpha=@n25.G
   )
   Alpha=@n26.B
  )
  n33 = MaterialFunctionCall!(
   Base Material=@n21
   Top Material=MaterialFunctionCall() {materialFunction=&/MF_LavaFalls}.Result
   Alpha=ComponentMask(None=TextureCoordinate() {coordinateIndex=3}) {g=true}
  ) {materialFunction=/Engine/Functions/MaterialLayerFunctions/MatLayerBlend_Standard}
  n34 = GetMaterialAttributes(None=StaticSwitchParameter(
    True=SetMaterialAttributes!(MaterialAttributes=@n22.MaterialAttributes, Opacity Mask=@n27, Emissive Color=@n32) {attributeSetTypes [#2, #1]}
    False=StaticSwitchParameter(True=@n33.Blended Material, False=@n21) {parameterName=IsLavaFalls}
   ) {parameterName=IsLavalStream}) {attributeGetTypes [#1]}
  n35 = TransformPosition(None=Constant3Vector()) {transformType=TRANSFORMPOSSOURCE_World}
  n36 = WorldPosition()
  n37 = StaticSwitchParameter(True=Multiply(A=Subtract(A=@n35, B=@n36.XYZ), B=OneMinus(None=@n18)), False=Constant3Vector()) {parameterName=IsRealBubble}
  n38 = Subtract(A=WorldPosition().XYZ, B=CameraPositionWS())
  n39 = Add(A=@n37, B=Multiply(A=Normalize(VectorInput=@n38), B=ScalarParameter() {parameterName=CameraOffsetPush}))
  MaterialFunctionCall(MA In=SetMaterialAttributes!(
    MaterialAttributes=@n34.MaterialAttributes
    World Position Offset=StaticSwitchParameter(True=@n39, False=@n37) {parameterName=UseZFightingFix}
    Emissive Color=@n34.MaterialAttributes
   ) {attributeSetTypes [#3, #1]}) {materialFunction=&MF_CreativeMoveTool}
  n40 = VectorParameter() {parameterName=TOD_EmissiveMult, channelNames {a=Night, b=Evening, g=Day, r=Morning}, defaultValue=<rgba 1 1 1 1>}
  n41 = VectorParameter() {parameterName=TOD_EmissiveMult_Mobile, channelNames {a=Night, b=Evening, g=Day, r=Morning}, defaultValue=<rgba 1 1 1 1>}
  Multiply(A=ShadingPathSwitch(
    Default=MaterialFunctionCall(Value_Morning=@n40.Morning, Value_Day=@n40.Morning, Value_Evening=@n40.Morning, Value_Night=@n40.Morning) {materialFunction=&Scalar_Time_of_Day}.Result
    Mobile=MaterialFunctionCall(Value_Morning=@n41.Morning, Value_Day=@n41.Morning, Value_Evening=@n41.Morning, Value_Night=@n41.Morning) {materialFunction=&Scalar_Time_of_Day}.Result
   ))
  OneMinus(None=@n17)
  VectorParameter() {parameterName=StreamColor, defaultValue=<rgba 5 1.1178 0.4948 1>}
 ]
}
