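# Unlit Nanite-enabled material for the large animated lava surface of the Vulcano environment set.
# Three panning world-space lava scales distort one another and fade by camera distance, feeding a gradient-mapped glow.
# A terrain heightmap drives brightness and glowing shoreline edges, with time-of-day emissive scaling.
# UseOceanblend layers in normal-blended lavafalls and a height-blended, foamy ocean with vertex offset, while mobile, forward and quality switches simplify the effect.
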Material /Game/Environments/Sets/Vulcano/Materials/M_LavaHuge
graph {
 properties {
  bUseMaterialAttributes=true
  bUsedWithNanite=true
  physMaterial=/Game/FortressPhysicalMaterials/Lava
  pixelDepthOffsetMode=PDOM_Legacy
  shadingModel=MSM_Unlit
  usedShadingModels=Unlit
 }
 nodes [
  n1 = ScalarParameter() {defaultValue=45, parameterName=FallsTilingX}
  n2 = ScalarParameter() {defaultValue=-4.0901, parameterName=FallsTilingY}
  n3 = Panner(
   Coordinate=Multiply(A=TextureCoordinate(), B=AppendVector(A=@n1, B=@n2))
   Speed=AppendVector(A=ScalarParameter() {defaultValue=-0.01, parameterName=FallsPanningX}, B=ScalarParameter() {defaultValue=0.03, parameterName=FallsPanningY})
  )
  n4 = Multiply(
   A=TextureCoordinate()
   B=AppendVector(A=ScalarParameter() {defaultValue=21, parameterName=FallsTilingX2}, B=ScalarParameter() {defaultValue=-5.7591, parameterName=FallsTilingY2})
  )
  n5 = AppendVector(
   A=ScalarParameter() {defaultValue=-0.0186, parameterName=FallsPanningX2}
   B=ScalarParameter() {defaultValue=0.05, parameterName=FallsPanningY2}
  )
  n6 = Multiply(
   A=TextureSample(UVs=@n3) {samplerType=SAMPLERTYPE_Masks, texture=/Game/Environments/Sets/Vulcano/Textures/LavaPileUp_M}.G
   B=TextureSample(UVs=Panner(Coordinate=@n4, Speed=@n5)) {samplerType=SAMPLERTYPE_Masks, texture=&LavaPileUp_M}.R
  )
  n7 = Divide(A=ComponentMask(None=WorldPosition().XYZ) {g=true, r=true}, B=ScalarParameter() {defaultValue=74780.3594, parameterName=LavaTilingHuge})
  n8 = VertexInterpolator(VS=Panner(Coordinate=@n7, Speed=ScalarParameter() {defaultValue=0.01, parameterName=PanSpeedHuge}))
  n9 = ScalarParameter() {defaultValue=0.05, parameterName=DistortHuge}
  n10 = Divide(A=ComponentMask(None=WorldPosition().XYZ) {g=true, r=true}, B=ScalarParameter() {defaultValue=9272.4404, parameterName=LavaTilingMedium})
  n11 = VertexInterpolator(VS=Panner(Coordinate=@n10, Speed=ScalarParameter() {defaultValue=0.01, parameterName=PanSpeedMedium}))
  n12 = ScalarParameter() {defaultValue=0.02, parameterName=DistortMedium}
  n13 = ScalarParameter() {defaultValue=-0.3952, parameterName=DistortSmall}
  n14 = TextureSample(UVs=@n11.PS) {samplerType=SAMPLERTYPE_Masks, texture=/Game/Environments/Sets/Vulcano/Textures/Lava_M}
  n15 = ComponentMask(None=WorldPosition().XYZ) {g=true, r=true}
  n16 = ScalarParameter() {defaultValue=909.5331, parameterName=LavaTilingSmall}
  n17 = VertexInterpolator(VS=Panner(Coordinate=Divide(A=@n15, B=@n16), Speed=ScalarParameter() {defaultValue=0.04, parameterName=PanSpeedSmall}))
  n18 = ShadingPathSwitch(Default=Add(A=Multiply(A=@n13, B=@n14.R), B=@n17.PS), Mobile=@n17.PS) {desc=Migrated from FeatureLevelSwitch}
  n19 = TextureSample(UVs=@n18) {samplerType=SAMPLERTYPE_Masks, texture=&Lava_M}
  n20 = ShadingPathSwitch(Default=Add(A=@n11.PS, B=Multiply(A=@n12, B=@n19.R)), Mobile=@n11.PS) {desc=Migrated from FeatureLevelSwitch}
  n21 = VertexInterpolator(VS=@n10)
  n22 = QualitySwitch(Default=@n20, Low=@n21.PS, High=@n20, Medium=@n21.PS)
  n23 = TextureSample(UVs=@n22) {samplerType=SAMPLERTYPE_Masks, texture=&Lava_M}
  n24 = ShadingPathSwitch(Default=Add(A=@n8.PS, B=Multiply(A=@n9, B=@n23.R)), Mobile=@n8.PS) {desc=Migrated from FeatureLevelSwitch}
  n25 = ShadingPathSwitch(Default=@n24, Deferred=@n24, Forward=@n8.PS, Mobile=@n8.PS)
  n26 = VertexInterpolator(VS=@n7)
  n27 = QualitySwitch(Default=@n25, Low=@n26.PS, High=@n25, Medium=@n26.PS)
  n28 = TextureSample(UVs=@n27) {samplerType=SAMPLERTYPE_Masks, texture=&Lava_M}
  n29 = Multiply(A=@n28.R, B=ScalarParameter() {defaultValue=3.1444, parameterName=BreakUpFalls})
  Add(A=@n6, B=@n29)
  LinearInterpolate(A=Constant() {r=1}, B=Constant() {r=0.85})
  MaterialFunctionCall() {materialFunction=&FlattenNormal}
  n30 = ComponentMask(None=WorldPosition() {worldPositionShaderOffset=WPT_ExcludeAllShaderOffsets}.XYZ) {b=true}
  n31 = ScalarParameter() {defaultValue=-24202.084, parameterName=OceanBlendOffset2}
  n32 = Clamp(None=Divide(A=Subtract(A=@n30, B=@n31), B=ScalarParameter() {defaultValue=27950.002, parameterName=OceanBlendDistance2}))
  Multiply(B=@n32)
  n33 = Divide(A=Multiply(A=@n19.R, B=ScalarParameter() {defaultValue=25, parameterName=PDO}), B=DotProduct(A=VertexNormalWS(), B=CameraVectorWS()))
  Multiply(A=@n33, B=MaterialFunctionCall() {materialFunction=&DitherTemporalAA}.Result)
  n34 = Power(Base=Multiply(A=@n6, B=@n29), Exp=ScalarParameter() {defaultValue=3.437, parameterName=LavaFallContrast})
  n35 = ObjectPositionWS()
  n36 = Subtract(A=Subtract(A=@n30, B=ComponentMask(None=@n35) {b=true}), B=ScalarParameter() {defaultValue=-13521.6211, parameterName=OceanBlendOffset3})
  n37 = Clamp(None=Divide(A=@n36, B=ScalarParameter() {defaultValue=28021.293, parameterName=OceanBlendDistance5}))
  Multiply(A=@n34, B=@n37)
  Multiply(A=@n19.R)
  n38 = Distance(A=CameraPositionWS(), B=WorldPosition().XYZ)
  n39 = Add(A=@n38, B=ScalarParameter() {parameterName=DistanceFadeOffsetSmall})
  n40 = ScalarParameter() {defaultValue=14785.1504, parameterName=DistanceFadeSmall}
  n41 = OneMinus(None=Clamp(None=Divide(A=@n39, B=@n40)))
  n42 = VertexInterpolator(VS=@n41)
  n43 = ShadingPathSwitch(Default=@n41, Mobile=@n42.PS) {desc=Migrated from FeatureLevelSwitch}
  n44 = ShadingPathSwitch(Default=@n43, Deferred=@n43, Forward=@n42.PS, Mobile=@n42.PS)
  n45 = Multiply(A=ScalarParameter() {defaultValue=2.103, parameterName=DetailGlowBrightness}, B=LinearInterpolate(A=@n23.G, B=@n19.G, Alpha=@n44))
  n46 = LinearInterpolate(A=@n28.G, B=@n23.G, Alpha=@n44)
  n47 = Multiply(
   A=ShadingPathSwitch(
    Default=VectorParameter() {parameterName=LavalGlowDetail, defaultValue=<rgba 20 0.8 0 1>}.RGB
    Mobile=VectorParameter() {parameterName=LavalGlowDetailMobile, defaultValue=<rgba 20 0.8 0 1>}.RGB
   )
   B=Power(Base=Multiply(A=@n45, B=@n46), Exp=ScalarParameter() {defaultValue=7.4823, parameterName=DetailGlowPower})
  )
  Multiply(A=@n44, B=@n47)
  n48 = VectorParameter() {parameterName=TOD_EmissiveMult, channelNames {a=Night, b=Evening, g=Day, r=Morning}, defaultValue=<rgba 1 1 1 1>}
  n49 = VectorParameter() {parameterName=TOD_EmissiveMult_Mobile, channelNames {a=Night, b=Evening, g=Day, r=Morning}, defaultValue=<rgba 1 1 1 1>}
  NamedRerouteDeclaration(None=ShadingPathSwitch(
    Default=MaterialFunctionCall(Value_Morning=@n48.Morning, Value_Day=@n48.Morning, Value_Evening=@n48.Morning, Value_Night=@n48.Morning) {materialFunction=&Scalar_Time_of_Day}.Result
    Mobile=MaterialFunctionCall(Value_Morning=@n49.Morning, Value_Day=@n49.Morning, Value_Evening=@n49.Morning, Value_Night=@n49.Morning) {materialFunction=&Scalar_Time_of_Day}.Result
   )) {name=TODMult, nodeColor=<rgba 0 1 0.6719 1>}
  ObjectPositionWS()
  OneMinus()
  n50 = Distance(
   A=ComponentMask(None=WorldPosition().XYZ) {g=true, r=true}
   B=ComponentMask(None=VectorParameter() {parameterName=GradientCenter, defaultValue=<rgba 29429.3984 -22007.0547 0 1>}.RGB) {g=true, r=true}
  )
  n51 = ScalarParameter() {defaultValue=195215.5938, parameterName=FalloffRadius}
  n52 = Power(Base=OneMinus(None=Divide(A=@n50, B=@n51)), Exp=ScalarParameter() {defaultValue=2.9444, parameterName=FalloffContrast})
  n53 = ScalarParameter() {parameterName=LavaBrightness2}
  n54 = ScalarParameter() {defaultValue=1.0746, parameterName=LavaBrightness}
  n55 = ScalarParameter() {defaultValue=-14488.0674, parameterName=TerrainHeight Min}
  n56 = ScalarParameter() {defaultValue=11082.7412, parameterName=TerrainHeightMax}
  n57 = ComponentMask(None=Subtract(A=WorldPosition().XYZ, B=Constant3Vector() {constant=<rgba -153000 -153000 -1692 0>})) {g=true, r=true}
  n58 = Multiply(A=ScalarParameter() {defaultValue=2708.7021, parameterName=DistortTerrainHeight}, B=Subtract(A=@n23.R) {constB=0.5})
  n59 = TextureSampleParameter2D(UVs=Divide(A=Add(A=@n57, B=@n58), B=ScalarParameter() {defaultValue=306000, parameterName=LandscapeSize})) {parameterName=HeightMapTerrain, samplerType=SAMPLERTYPE_LinearGrayscale, texture=/Game/Environments/Sets/Vulcano/Textures/T_TerrainHeightMap}
  n60 = ComponentMask(None=ActorPositionWS()) {b=true}
  n61 = Subtract(A=LinearInterpolate(A=@n55, B=@n56, Alpha=@n59.R), B=@n60)
  n62 = TextureSampleParameter2D(UVs=Divide(A=@n57, B=ScalarParameter() {defaultValue=306000, parameterName=LandscapeSize})) {parameterName=HeightMapTerrain, samplerType=SAMPLERTYPE_LinearGrayscale, texture=&T_TerrainHeightMap}
  n63 = Subtract(A=LinearInterpolate(A=@n55, B=@n56, Alpha=@n62.R), B=@n60)
  n64 = ScalarParameter() {defaultValue=18909.5664, parameterName=BlendDistanceTerrain}
  n65 = ScalarParameter() {defaultValue=1311.1791, parameterName=BlendDistanceTerrain2}
  n66 = ActorPositionWS()
  n67 = Divide(
   A=Subtract(A=ComponentMask(None=@n66) {b=true}, B=ScalarParameter() {defaultValue=2210.1787, parameterName=MaxObjectPosOffset})
   B=ScalarParameter() {defaultValue=11288.4316, parameterName=MaxObjectPosHeight}
  )
  n68 = VertexInterpolator(VS=OneMinus(None=Clamp(None=@n67)))
  n69 = Abs(None=Divide(A=ShadingPathSwitch(Default=@n61, Mobile=@n63) {desc=Migrated from FeatureLevelSwitch}, B=LinearInterpolate(A=@n64, B=@n65, Alpha=@n68.PS)))
  n70 = Clamp(None=Power(Base=@n69, Exp=ScalarParameter() {defaultValue=1.9827, parameterName=LavaBrightness_Power}))
  n71 = Multiply(A=Time(), B=ScalarParameter() {defaultValue=-0.02, parameterName=GradientMotionSpeed})
  n72 = ScalarParameter() {defaultValue=0.3, parameterName=GradientMotionWaveLength}
  n73 = Sine(None=Frac(None=Divide(A=@n71, B=@n72)))
  n74 = ScalarParameter() {defaultValue=0.1, parameterName=GradientMotionDistance}
  n75 = OneMinus(None=@n69)
  n76 = ScalarParameter() {defaultValue=1.5204, parameterName=GradientTerrainContribution}
  n77 = Clamp(None=OneMinus(None=Multiply(A=@n75, B=@n76)))
  n78 = Add(A=@n38, B=ScalarParameter() {defaultValue=-5000, parameterName=DistanceFadeOffsetMedium})
  n79 = ScalarParameter() {defaultValue=103232.75, parameterName=DistanceFadeMedium}
  n80 = OneMinus(None=Clamp(None=Divide(A=@n78, B=@n79)))
  n81 = VertexInterpolator(VS=@n80)
  n82 = ShadingPathSwitch(Default=@n80, Mobile=@n81.PS) {desc=Migrated from FeatureLevelSwitch}
  n83 = LinearInterpolate(A=Constant() {r=1}, B=@n23.R, Alpha=ShadingPathSwitch(Default=@n82, Deferred=@n82, Forward=@n81.PS, Mobile=@n81.PS))
  n84 = ScalarParameter() {parameterName=MediumContribution}
  n85 = Multiply(A=@n28.R, B=Clamp(None=Add(A=@n83, B=@n84)))
  n86 = ShadingPathSwitch(
   Default=Add(A=Multiply(A=@n73, B=@n74), B=Multiply(A=@n77, B=@n85))
   Mobile=Multiply(A=@n83, B=ScalarParameter() {defaultValue=0.6571, parameterName=MobileDarken})
  ) {desc=Migrated from FeatureLevelSwitch}
  n87 = StaticSwitchParameter(True=ShadingPathSwitch(Default=Multiply(A=@n86, B=@n37), Mobile=@n86) {desc=Migrated from FeatureLevelSwitch}, False=@n86) {parameterName=UseOceanblend}
  n88 = OneMinus(None=TextureSample(UVs=Clamp(None=@n87) {clampMode=CMODE_ClampMin, minDefault=0.02}) {samplerType=SAMPLERTYPE_Masks, texture=/Game/Environments/Sets/Vulcano/Textures/Lava_Gradient}.R)
  n89 = ScalarParameter() {defaultValue=4.3659, parameterName=LavaContrast2}
  n90 = Constant() {r=1}
  n91 = Power(Base=@n19.R, Exp=ScalarParameter() {defaultValue=5.6039, parameterName=LavaContrast})
  n92 = LinearInterpolate(
   A=VectorParameter() {parameterName=LavalGlow_1, defaultValue=<rgba 0.1146 0.008 0 1>}.RGB
   B=Multiply(A=LinearInterpolate(A=@n53, B=@n54, Alpha=@n70), B=VectorParameter() {parameterName=LavalGlow_2, defaultValue=<rgba 7 0.5 0 1>}.RGB)
   Alpha=Multiply(A=Power(Base=@n88, Exp=@n89), B=LinearInterpolate(A=@n90, B=@n91, Alpha=@n44))
  )
  n93 = VectorParameter() {parameterName=LavalGlowEdgeBright, defaultValue=<rgba 20 0.8 0 1>}
  n94 = VectorParameter() {parameterName=LavalGlowEdgeBrightMobile, defaultValue=<rgba 20 0.8 0 1>}
  n95 = Divide(A=@n63, B=ScalarParameter() {defaultValue=1045.0656, parameterName=BlendDistanceTerrainShort})
  n96 = Clamp(None=OneMinus(None=Abs(None=@n95)))
  n97 = Multiply(A=@n96, B=ScalarParameter() {defaultValue=0.7594, parameterName=TranstionHeightBlendEdge})
  n98 = ComponentMask(None=WorldPosition().XYZ) {g=true, r=true}
  n99 = ScalarParameter() {defaultValue=968.9221, parameterName=LavaEdgeTiling1}
  n100 = TextureSample(UVs=Panner(Coordinate=Divide(A=@n98, B=@n99), Speed=ScalarParameter() {defaultValue=0.1143, parameterName=PanSpeedEdgeGlow})) {samplerType=SAMPLERTYPE_Masks, texture=/Game/Environments/Sets/Vulcano/Textures/LavaGround_M}
  n101 = Divide(A=@n98, B=ScalarParameter() {defaultValue=1688.7714, parameterName=LavaEdgeTiling2})
  n102 = TextureSample(UVs=Panner(
    Coordinate=MaterialFunctionCall(XY=@n101) {materialFunction=&Swizzle}.YX
    Speed=ScalarParameter() {defaultValue=-0.1143, parameterName=PanSpeedEdgeGlow2}
   )) {samplerType=SAMPLERTYPE_Masks, texture=&LavaGround_M}
  n103 = Clamp(None=OneMinus(None=Add(A=@n100.G, B=@n102.G))) {clampMode=CMODE_ClampMin}
  n104 = ScalarParameter() {defaultValue=1.5753, parameterName=ContrastEdgeBlend}
  n105 = LinearInterpolate(
   A=ShadingPathSwitch(Default=Multiply(A=@n52, B=@n92), Mobile=@n92) {desc=Migrated from FeatureLevelSwitch}
   B=LinearInterpolate(
    A=VectorParameter() {parameterName=LavaGlowEdgeDark, defaultValue=<rgba 0 0 0 1>}.RGB
    B=ShadingPathSwitch(Default=@n93.RGB, Mobile=@n94.RGB)
    Alpha=MaterialFunctionCall(Transition Phase=@n97, Height Texture=@n103, Contrast=@n104) {materialFunction=&HeightLerp}.Alpha
   )
   Alpha=@n96
  )
  n106 = Add(A=@n105, B=@n47)
  Reroute(None=@n106)
  n107 = VectorParameter() {parameterName=MobileLavaEmissiveAdd, defaultValue=<rgba 0.2031 0 0 1>}
  n108 = Multiply(A=NamedRerouteUsage(), B=ShadingPathSwitch(Default=@n106, Mobile=Add(A=@n107.RGB, B=@n106)))
  Reroute(None=@n108)
  n109 = ComponentMask(None=Divide(A=WorldPosition().XYZ, B=Constant() {r=18286.0156})) {g=true, r=true}
  n110 = Multiply(A=Time(), B=Constant() {r=0.6})
  n111 = Panner(Coordinate=Multiply(A=Constant() {r=0.2557}, B=@n109), Time=@n110) {bFractionalPart=true, speedX=0.01, speedY=-0.005}
  Rotator(Coordinate=@n111, Time=Constant() {r=0.8658})
  ScalarParameter() {defaultValue=0.2095, parameterName=LavaBrightnessold}
  n112 = VectorParameter() {parameterName=OceanFoam, defaultValue=<rgba 0.6615 0.6615 0.6615 1>}
  n113 = VectorParameter() {parameterName=OceanFoam2, defaultValue=<rgba 0.0885 0.0885 0.0885 1>}
  n114 = ScalarParameter() {defaultValue=4320.4111, parameterName=OceanBlendDistance4}
  n115 = Multiply(A=Clamp(None=Divide(A=@n36, B=@n114)), B=@n32)
  n116 = MaterialFunctionCall(
   Transition Phase=ShadingPathSwitch(Default=@n115, Mobile=VertexInterpolator(VS=@n115).PS) {desc=Migrated from FeatureLevelSwitch}
   Height Texture=Multiply(A=@n28.R, B=@n23.R)
   Contrast=ScalarParameter() {parameterName=FoamBlendContrast}
  ) {materialFunction=&HeightLerp}
  n117 = ComponentMask(None=WorldPosition().XYZ) {g=true, r=true}
  n118 = Constant2Vector() {g=-153000, r=-153000}
  n119 = TextureSample(UVs=Divide(A=Subtract(A=@n117, B=@n118), B=Constant() {r=306200})) {
   samplerSource=SSM_Clamp_WorldGroupSettings
   samplerType=SAMPLERTYPE_LinearColor
   texture=/Game/Athena/Prototype/Terrain/Textures/T_Athena_Terrain_Mask
  }
  n120 = Constant() {r=0.2}
  n121 = LinearInterpolate(
   A=VectorParameter() {parameterName=deepwater, defaultValue=<rgba 0.0049 0.0249 0.05 1>}.RGB
   B=VectorParameter() {parameterName=shallowwater, defaultValue=<rgba 0.0469 0.5 0.4635 1>}.RGB
   Alpha=Multiply(A=Clamp(None=@n119.A, Max=@n120) {clampMode=CMODE_ClampMin, maxDefault=0.5}, B=VertexColor().A)
  )
  n122 = ComponentMask(None=Divide(A=WorldPosition().XYZ, B=Constant() {r=6202.6328})) {g=true, r=true}
  n123 = Time()
  n124 = TextureSample(UVs=Panner(Coordinate=@n122, Time=@n123) {bFractionalPart=true, speedX=-0.03, speedY=0.03}) {texture=/Game/Global/Textures/Default/Tileable_Patterns/Foam}
  n125 = TextureSample(UVs=Panner(Coordinate=Multiply(A=@n122) {constB=1.5}, Time=@n123) {bFractionalPart=true, speedX=0.02, speedY=-0.1}) {texture=&Foam}
  n126 = Multiply(A=Constant3Vector() {constant=<rgba 1.5469 2.5102 3 0>}, B=Clamp(None=Subtract(A=@n124.G, B=@n125.G)))
  n127 = Constant() {r=3}
  n128 = Clamp(None=Subtract(A=Multiply(A=@n119.A, B=@n127)) {constB=0.6})
  n129 = QualitySwitch(Default=@n121, Low=@n121, High=Add(A=@n121, B=Multiply(A=@n126, B=@n128)), Medium=@n121)
  n130 = ComponentMask(None=WorldPosition().XYZ) {g=true, r=true}
  n131 = ScalarParameter() {defaultValue=478490.3125, parameterName=TextureSize_1}
  n132 = TextureSample(UVs=VertexInterpolator(VS=Divide(A=@n130, B=@n131)).PS) {samplerType=SAMPLERTYPE_Masks, texture=/Game/Environments/World/Backgrounds/Transylvania/Textures/T_OceanGround_M}
  n133 = ScalarParameter() {defaultValue=0.2248, parameterName=SandBlend}
  n134 = ScalarParameter() {defaultValue=-0.3143, parameterName=SandBlend_1}
  n135 = LinearInterpolate(
   A=LinearInterpolate(
    A=VectorParameter() {parameterName=Shallow, defaultValue=<rgba 0.0306 0.1031 0.151 1>}.RGB
    B=ShadingPathSwitch(Default=@n129, Mobile=@n121) {desc=Migrated from FeatureLevelSwitch}
    Alpha=@n132.G
   )
   B=VectorParameter() {parameterName=Ground, defaultValue=<rgba 0.1389 0.3107 0.375 1>}.RGB
   Alpha=Clamp(None=LinearInterpolate(A=@n133, B=@n134, Alpha=@n132.G))
  )
  n136 = ShadingPathSwitch(
   Default=LinearInterpolate(A=@n135, B=LinearInterpolate(A=@n112.RGB, B=@n113.RGB, Alpha=@n116.Lerp Alpha No Contrast), Alpha=@n116.Alpha)
   Mobile=@n135
  ) {desc=Migrated from FeatureLevelSwitch}
  n137 = ScalarParameter() {parameterName=MetallicOcean}
  n138 = ScalarParameter() {defaultValue=0.1, parameterName=SpecularOcean}
  n139 = ScalarParameter() {defaultValue=0.2, parameterName=RoughnessOcean}
  n140 = Constant()
  n141 = Multiply(A=Constant() {r=0.119}, B=@n109)
  n142 = MaterialFunctionCall(
   BaseNormal=TextureSample(UVs=Panner(Coordinate=@n141, Time=@n110) {bFractionalPart=true, speedX=-0.015, speedY=0.005}) {samplerType=SAMPLERTYPE_Normal, texture=/Game/Global/Textures/Default/Tileable_Patterns/T_Water2_N}.RGB
   AdditionalNormal=TextureSample(UVs=@n111) {samplerType=SAMPLERTYPE_Normal, texture=/Game/Effects/Fort_Effects/Textures/WaterFoam0001_19_N}.RGB
  ) {materialFunction=&BlendAngleCorrectedNormals}
  n143 = MaterialFunctionCall(Fade Length=Constant() {r=185195.2031}, Fade Offset=Constant()) {materialFunction=&CameraDepthFade}
  n144 = LinearInterpolate(A=Constant() {r=1}, B=Constant() {r=0.85}, Alpha=@n143.Result)
  n145 = Time()
  n146 = Multiply(A=@n109, B=Constant() {r=1})
  n147 = Add(
   A=TextureSample(UVs=Panner(Coordinate=@n109, Time=@n145) {bFractionalPart=true, speedX=0.009, speedY=-0.009}) {samplerType=SAMPLERTYPE_Normal, texture=&T_Water2_N}.RGB
   B=TextureSample(UVs=Panner(Coordinate=@n146, Time=@n145) {bFractionalPart=true, speedX=-0.003, speedY=0.006}) {samplerType=SAMPLERTYPE_Normal, texture=&WaterFoam0001_19_N}.RGB
  )
  n148 = LinearInterpolate(A=Constant() {r=0.2}, B=Constant() {r=1}, Alpha=@n143.Result)
  n149 = Fresnel(ExponentIn=Constant() {r=4}, Normal=VertexNormalWS()) {baseReflectFraction=0}
  n150 = MaterialFunctionCall(
   Normal=Add(
    A=MaterialFunctionCall(Normal=@n142.Result, Flatness=@n144) {materialFunction=&FlattenNormal}.Result
    B=MaterialFunctionCall(Normal=@n147, Flatness=@n148) {materialFunction=&FlattenNormal}.Result
   )
   Flatness=@n149
  ) {materialFunction=&FlattenNormal}
  n151 = MaterialFunctionCall(Normal=@n142.Result, Flatness=LinearInterpolate(A=Constant() {r=0.85}, B=Constant() {r=1}, Alpha=@n149)) {materialFunction=&FlattenNormal}
  n152 = Constant3Vector() {constant=<rgba 0 0 1 0>}
  n153 = ShadingPathSwitch(
   Default=ShadingPathSwitch(Default=QualitySwitch(Default=@n150.Result, Low=@n151.Result, High=@n150.Result, Medium=@n150.Result), Mobile=@n151.Result) {desc=Migrated from FeatureLevelSwitch}
   Mobile=@n152
  ) {desc=Migrated from FeatureLevelSwitch}
  n154 = Transform(None=ShadingPathSwitch(Default=@n153, Deferred=@n153, Forward=@n152, Mobile=@n152)) {transformSourceType=TRANSFORMSOURCE_World, transformType=TRANSFORM_Tangent}
  n155 = LinearInterpolate(
   A=VectorParameter() {parameterName=LavaFallColor, defaultValue=<rgba 0 0 0 1>}.RGB
   B=VectorParameter() {parameterName=LavaFallColor2, defaultValue=<rgba 5 0.3913 0 1>}.RGB
   Alpha=@n34
  )
  n156 = TextureSample(UVs=@n3) {samplerType=SAMPLERTYPE_Normal, texture=/Game/Environments/Sets/Vulcano/Textures/LavaPileUp_N}
  n157 = Constant3Vector() {constant=<rgba 0 0 1 0>}
  n158 = MakeMaterialAttributes!(
   BaseColor=@n155
   Metallic
   Specular
   Roughness=ScalarParameter() {defaultValue=0.8571, parameterName=RoughnessFalls}
   Anisotropy
   EmissiveColor=@n155
   Opacity
   OpacityMask
   Normal=ShadingPathSwitch(Default=@n156.RGB, Deferred=@n156.RGB, Forward=@n157, Mobile=@n157)
   Tangent
   WorldPositionOffset
   SubsurfaceColor
   ClearCoat
   ClearCoatRoughness
   AmbientOcclusion
   Refraction
   CustomizedUVs_0
   CustomizedUVs_1
   CustomizedUVs_2
   CustomizedUVs_3
   CustomizedUVs_4
   CustomizedUVs_5
   CustomizedUVs_6
   CustomizedUVs_7
   PixelDepthOffset
   ShadingModel
   Displacement
  )
  n159 = VertexNormalWS()
  n160 = Clamp(None=LinearInterpolate(
    A=ScalarParameter() {defaultValue=2.8996, parameterName=LavaFallBlend1}
    B=ScalarParameter() {parameterName=LavaFallBlend2}
    Alpha=ComponentMask(None=@n159) {b=true}
   ))
  n161 = ScalarParameter() {parameterName=FlattenNormalHuge}
  n162 = TextureSample(UVs=@n27) {samplerType=SAMPLERTYPE_Normal, texture=/Game/Environments/Sets/Vulcano/Textures/Lava_N}
  n163 = TextureSample(UVs=@n22) {samplerType=SAMPLERTYPE_Normal, texture=&Lava_N}
  n164 = TextureSample(UVs=@n18) {samplerType=SAMPLERTYPE_Normal, texture=&Lava_N}
  n165 = Constant3Vector() {constant=<rgba 0 0 1 0>}
  n166 = ShadingPathSwitch(
   Default=MaterialFunctionCall(
    BaseNormal=MaterialFunctionCall(Normal=@n162.RGB, Flatness=@n161) {materialFunction=&FlattenNormal}.Result
    AdditionalNormal=LinearInterpolate(A=@n163.RGB, B=@n164.RGB, Alpha=@n44)
   ) {materialFunction=&BlendAngleCorrectedNormals}.Result
   Mobile=@n165
  ) {desc=Migrated from FeatureLevelSwitch}
  n167 = ShadingPathSwitch(Default=@n166, Deferred=@n166, Forward=@n165, Mobile=@n165)
  n168 = MakeMaterialAttributes(
   BaseColor=@n107.RGB
   Metallic=ScalarParameter() {defaultValue=0.0571, parameterName=M}
   Specular=ScalarParameter() {parameterName=S}
   Roughness=LinearInterpolate(A=ScalarParameter() {defaultValue=0.0116, parameterName=R}, B=ScalarParameter() {defaultValue=0.7099, parameterName=R2}, Alpha=@n19.R)
   EmissiveColor=@n108
   Normal=QualitySwitch(Default=@n167, Low=@n165, High=@n167, Medium=@n165)
  )
  n169 = Subtract(A=ComponentMask(None=WorldPosition().XYZ) {b=true}, B=ScalarParameter() {defaultValue=-5093.8433, parameterName=OceanLevel})
  n170 = Abs(None=Clamp(None=@n169) {clampMode=CMODE_ClampMax, maxDefault=0})
  n171 = Add(A=@n30, B=@n170)
  n172 = ScalarParameter() {defaultValue=-5070.5273, parameterName=OceanBlendOffset}
  n173 = Clamp(None=Divide(A=Subtract(A=@n171, B=@n172), B=ScalarParameter() {defaultValue=2438.0903, parameterName=OceanBlendDistance}))
  n174 = VertexInterpolator(VS=@n173)
  n175 = ShadingPathSwitch(
   Default=MaterialFunctionCall!(
    Base Material=MakeMaterialAttributes(BaseColor=@n136, Metallic=@n137, Specular=@n138, Roughness=@n139, EmissiveColor=@n140, Normal=@n154)
    Top Material=MaterialFunctionCall(Base Material=@n168, Top Material=@n158, Alpha=@n160) {materialFunction=/Engine/Functions/MaterialLayerFunctions/MatLayerBlend_Standard}.Blended Material
    Alpha=ShadingPathSwitch(Default=@n173, Deferred=@n173, Forward=@n174.PS, Mobile=@n174.PS)
   ) {materialFunction=&MatLayerBlend_Standard}.Blended Material
   Mobile=@n168
  ) {desc=Migrated from FeatureLevelSwitch}
  n176 = Constant()
  n177 = Constant()
  n178 = ShadingPathSwitch(Default=@n33, Mobile=@n177) {desc=Migrated from FeatureLevelSwitch}
  n179 = ShadingPathSwitch(Default=@n178, Deferred=@n178, Forward=@n177, Mobile=@n177)
  n180 = Normalize(VectorInput=Subtract(A=CameraPositionWS(), B=WorldPosition().XYZ))
  n181 = LinearInterpolate(
   A=ScalarParameter() {defaultValue=-50, parameterName=PushUp}
   B=ScalarParameter() {defaultValue=-500, parameterName=PushZfightFixFar}
   Alpha=MaterialFunctionCall(Distance=ScalarParameter() {defaultValue=7000, parameterName=PushUpFade}) {materialFunction=&Near_Camera_Fade}.Result
  )
  n182 = Add(A=Multiply(A=Multiply(A=@n180, B=@n181), B=Clamp(None=@n169) {maxDefault=0, minDefault=-1}), B=AppendVector(A=Constant2Vector(), B=@n170))
  n183 = Constant3Vector()
  n184 = GetMaterialAttributes(None=SetMaterialAttributes!(
    MaterialAttributes=StaticSwitchParameter(True=@n175, False=@n168) {parameterName=UseOceanblend}
    World Position Offset=ShadingPathSwitch(Default=@n182, Mobile=@n183) {desc=Migrated from FeatureLevelSwitch}
    Pixel Depth Offset=StaticSwitchParameter(True=@n176, False=@n179) {parameterName=UseOceanblend}
   ) {attributeSetTypes [#1, #2]}) {attributeGetTypes [#3]}
  SetMaterialAttributes!(
   MaterialAttributes=@n184.MaterialAttributes
   Emissive Color=MaterialFunctionCall(Emissive=@n184.Emissive Color) {materialFunction=&MF_FX_EmissiveAdjustments}.Result
  ) {attributeSetTypes [#3]}
  n185 = Constant()
  n186 = ScalarParameter() {defaultValue=0.0519, parameterName=Twist_bias}
  n187 = ScalarParameter() {defaultValue=9.7199, parameterName=Twist_Mod}
  n188 = ScalarParameter() {defaultValue=0.5, parameterName=Twist_SineOffset}
  n189 = ScalarParameter() {defaultValue=1, parameterName=Twist_SineSpeed}
  n190 = Multiply(A=@n109, B=ScalarParameter() {defaultValue=0.175, parameterName=Twist_Size})
  n191 = ScalarParameter() {defaultValue=0.2385, parameterName=Twist_RotationMod}
  n192 = ScalarParameter() {defaultValue=1, parameterName=Twist_Time}
  n193 = AppendVector(
   A=ScalarParameter() {defaultValue=-0.6433, group=Twist, parameterName=GridNoiseTileRep}
   B=ScalarParameter() {defaultValue=136.6535, group=Twist, parameterName=GridNoiseTileRep2}
  )
  n194 = ScalarParameter() {defaultValue=1.6632, parameterName=Twist_Falloff}
  ShadingPathSwitch(
   Default=QualitySwitch(
    Default=@n185
    Low=@n185
    High=MaterialFunctionCall!(
     Result Bias=@n186
     Result Modulation=@n187
     SineAnimOffset=@n188
     SineAnimSpeed=@n189
     UVS * size=@n190
     Rotation amount modulator=@n191
     Time=@n192
     Flow Direction
     Grid Noise Tile Rep=@n193
     Twist Falloff=@n194
    ) {materialFunction=/Game/Environments/Sets/Arid/Materials/Fortnite_GroundTwistFunction}.Distortion
    Medium=@n185
   )
   Mobile=Constant()
  ) {desc=Migrated from FeatureLevelSwitch}
  n195 = ShadingPathSwitch(Default=@n106, Mobile=@n105) {desc=Migrated from FeatureLevelSwitch}
  ShadingPathSwitch(Default=@n195, Deferred=@n195, Forward=@n105, Mobile=@n105)
  Time()
  VectorParameter() {parameterName=Deep, defaultValue=<rgba 0 0 0 1>}
 ]
}
