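# Material function for the open waterfall water materials that produces tiled, scrolling UV coordinates.
# The U and V texture coordinates are each multiplied by their own tiling input, which defaults to 1.
# Time multiplied by a separate pan speed per axis is added to each channel, then both are appended into one UV output.
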MaterialFunction /Game/Packages/DS_Fortnite_Terrain_NoLOD/Water/OpenWaterFall/MaterialFunctions/MF_WaterFall_Pann
graph {nodes [
  n1 = FunctionInput(Preview=Constant() {r=1}) {bUsePreviewValueAsDefault=true, description="Default = 1", inputName=X_Tile, inputType=FunctionInput_Scalar, sortPriority=1}
  n2 = Multiply(A=ComponentMask(None=TextureCoordinate()) {r=true}, B=@n1)
  n3 = Time()
  n4 = Multiply(
   A=@n3
   B=FunctionInput(Preview=Constant()) {bUsePreviewValueAsDefault=true, inputName=X_Pann, inputType=FunctionInput_Scalar, sortPriority=2}
  )
  n5 = FunctionInput(Preview=Constant() {r=1}) {bUsePreviewValueAsDefault=true, description="Default = 1", inputName=Y - Tiling, inputType=FunctionInput_Scalar, sortPriority=3}
  n6 = Multiply(A=ComponentMask(None=TextureCoordinate()) {g=true}, B=@n5)
  n7 = Multiply(
   A=@n3
   B=FunctionInput(Preview=Constant()) {bUsePreviewValueAsDefault=true, inputName=Y_Pann, inputType=FunctionInput_Scalar, sortPriority=4}
  )
  FunctionOutput(None=AppendVector(A=Add(A=@n2, B=@n4), B=Add(A=@n6, B=@n7))) {outputName=None}
 ]}
