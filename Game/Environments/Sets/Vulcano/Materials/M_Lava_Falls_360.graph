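# Lava falls material for the Vulcano set that blends glowing, flowing lava into a world-projected magma base with optional bubbles.
# Static switch IsBottomBlend picks either panning falls with height-lerped edges and gradient, or a radially mapped pile-up for the bottom, each layered over the magma by UV-channel masks.
# World position offset adds wavy motion plus vertex-red-driven pile lift or sine rings, and radial direction vectors also feed magma UV distortion.
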Material /Game/Environments/Sets/Vulcano/Materials/M_Lava_Falls_360
graph {
 properties {bUseMaterialAttributes=true, pixelDepthOffsetMode=PDOM_Legacy, usedShadingModels=Default Lit}
 nodes [
  n1 = WorldPosition() {worldPositionShaderOffset=WPT_ExcludeAllShaderOffsets}
  n2 = ComponentMask(None=TransformPosition(None=@n1.XYZ) {transformSourceType=TRANSFORMPOSSOURCE_World}) {g=true, r=true}
  n3 = AppendVector(
   A=ScalarParameter() {defaultValue=444.3143, parameterName=RadialSize2}
   B=ScalarParameter() {defaultValue=464.9501, parameterName=RadialSizeY2}
  )
  n4 = Divide(A=@n2, B=@n3)
  n5 = ScalarParameter() {defaultValue=2, parameterName=RadialTilingX2}
  n6 = ScalarParameter() {defaultValue=-3.8, parameterName=RadialTilingY2}
  n7 = Time()
  n8 = AppendVector(A=ScalarParameter() {defaultValue=0.05, parameterName=Speed_x2}, B=ScalarParameter() {defaultValue=0.1429, parameterName=Speed_y2})
  n9 = Add(
   A=Multiply(
    A=MaterialFunctionCall(Vector or UVs=@n4) {materialFunction=&VectorToRadialValue}.Radial Coordinates
    B=AppendVector(A=@n5, B=@n6)
   )
   B=Frac(None=Multiply(A=@n7, B=@n8))
  )
  n10 = TextureCoordinate()
  n11 = DDX(Value=@n10)
  n12 = DDY(Value=@n10)
  n13 = TextureSample!(UVs=@n9, Tex, "DDX(UVs)"=@n11, "DDY(UVs)"=@n12, Apply View MipBias) {mipValueMode=TMVM_Derivative, samplerType=SAMPLERTYPE_Masks, texture=/Game/Environments/Sets/Vulcano/Textures/LavaPileUp_M}
  n14 = ComponentMask(None=TextureCoordinate() {coordinateIndex=1}) {g=true}
  Add(A=@n13.G, B=@n14)
  n15 = Multiply(
   A=TextureSample!(UVs=@n9, Tex, Apply View MipBias) {samplerType=SAMPLERTYPE_Masks, texture=&LavaPileUp_M}.G
   B=Constant3Vector() {constant=<rgba 0 0 1 0>}
  )
  n16 = VertexColor()
  n17 = Multiply(A=Add(A=Multiply(A=@n15, B=@n16.R), B=@n16.R), B=ScalarParameter() {defaultValue=18.066, parameterName=PileOffsetWPO})
  n18 = Subtract(A=WorldPosition().XYZ, B=TransformPosition(None=Constant3Vector()) {transformType=TRANSFORMPOSSOURCE_World})
  n19 = ComponentMask(None=TextureCoordinate() {coordinateIndex=1}) {g=true}
  n20 = Multiply(
   A=Multiply(
    A=MaterialFunctionCall(Vector=@n18) {materialFunction=&SafeNormalize}.Result
    B=OneMinus(None=@n19)
   )
   B=ScalarParameter() {parameterName=BlendDiscSize}
  )
  Add(A=@n17, B=@n20)
  n21 = Constant() {r=1.5}
  n22 = Subtract(A=ScalarParameter() {defaultValue=-0.5, group=Displacement, parameterName=Displace_Concave}, B=@n21)
  n23 = Subtract(A=ScalarParameter() {defaultValue=2, group=Displacement, parameterName=Displace_Convex}, B=@n21)
  n24 = MaterialFunctionCall!(
   X_Tile=ScalarParameter() {defaultValue=-1, group=UV_Pann, parameterName=TilingX_Left}
   X_Pann=ScalarParameter() {defaultValue=0.1, group=UV_Pann, parameterName=PanningX_Left}
   Y - Tiling=ScalarParameter() {defaultValue=1.65, group=UV_Pann, parameterName=TilingY_Left}
   Y_Pann=ScalarParameter() {defaultValue=-0.84, group=UV_Pann, parameterName=PanningY_Left}
  ) {materialFunction=/Game/Packages/DS_Fortnite_Terrain_NoLOD/Water/OpenWaterFall/MaterialFunctions/MF_WaterFall_Pann}
  n25 = TextureSampleParameter2D(UVs=AppendVector(A=ComponentMask(None=@n24) {r=true}, B=ComponentMask(None=@n24) {g=true})) {parameterName=Param_1, texture=/Game/Packages/Fortnite_UniqueMaterials_1/Textures/FoamHeight}
  n26 = MaterialFunctionCall(
   X_Tile=ScalarParameter() {defaultValue=1, group=UV_Pann, parameterName=TilingX_Right}
   X_Pann=ScalarParameter() {defaultValue=0.05, group=UV_Pann, parameterName=PanningX_Right}
   Y - Tiling=ScalarParameter() {defaultValue=2, group=UV_Pann, parameterName=TilingY_Right}
   Y_Pann=ScalarParameter() {defaultValue=-0.77, group=UV_Pann, parameterName=PanningY_Right}
  ) {materialFunction=&MF_WaterFall_Pann}
  n27 = TextureSampleParameter2D(UVs=AppendVector(A=ComponentMask(None=@n26) {r=true}, B=ComponentMask(None=@n26) {g=true})) {parameterName=Param, texture=&FoamHeight}
  n28 = ScalarParameter() {defaultValue=3, group=Color_adjust, parameterName=FoamContribution_Top}
  n29 = LinearInterpolate(A=@n28, B=ScalarParameter() {defaultValue=0.75, group=Color_adjust, parameterName=FoamContribution_Bottom, sortPriority=1})
  n30 = ComponentMask(None=Power(Base=Multiply(A=@n25.RGB, B=@n27.RGB), Exp=LinearInterpolate(A=@n28, B=@n29)) {constExponent=1.25}) {r=true}
  n31 = Multiply(
   A=Multiply(A=LinearInterpolate(A=@n22, B=@n23, Alpha=@n30) {constA=-0.5}, B=VertexNormalWS())
   B=LinearInterpolate(
    A=ScalarParameter() {defaultValue=100, group=Displacement, parameterName=vert displace_min}
    B=ScalarParameter() {defaultValue=200, group=Displacement, parameterName=vert_displacemax}
   )
  )
  n32 = VertexColor()
  n33 = VectorParameter() {parameterName=Location, defaultValue=<rgba 0 51.1395 0 1>}
  n34 = ComponentMask(None=Subtract(A=WorldPosition().XYZ, B=TransformPosition(None=@n33.RGB) {transformType=TRANSFORMPOSSOURCE_World})) {g=true, r=true}
  n35 = Divide(
   A=@n34
   B=AppendVector(
    A=ScalarParameter() {defaultValue=458.8751, parameterName=RadialSize}
    B=ScalarParameter() {defaultValue=319.8525, parameterName=RadialSizeY}
   )
  )
  n36 = ScalarParameter() {defaultValue=0.9198, parameterName=RadialTilingX}
  n37 = ScalarParameter() {defaultValue=0.4531, parameterName=RadialTilingY}
  n38 = Time()
  n39 = AppendVector(A=ScalarParameter() {parameterName=Speed_x}, B=ScalarParameter() {defaultValue=0.06, parameterName=Speed_y})
  n40 = Add(
   A=Multiply(
    A=MaterialFunctionCall(Vector or UVs=@n35) {materialFunction=&VectorToRadialValue}.Radial Coordinates
    B=AppendVector(A=@n36, B=@n37)
   )
   B=Frac(None=Multiply(A=@n38, B=@n39))
  )
  n41 = ComponentMask(None=@n40) {g=true}
  n42 = ScalarParameter() {defaultValue=4, parameterName=SineWaveFrequency}
  n43 = Add(A=Sine(None=Multiply(A=@n41, B=@n42)))
  n44 = Multiply(
   A=Multiply(A=ScalarParameter() {defaultValue=2.9461, parameterName=WPORingsOffset}, B=@n32.R)
   B=Multiply(A=@n43, B=Constant3Vector() {constant=<rgba 0 0 1 0>})
  )
  n45 = Add(A=WorldPosition().XYZ, B=Multiply(A=Time(), B=ScalarParameter() {defaultValue=35.3185, parameterName=WaveySpeed}))
  n46 = ScalarParameter() {defaultValue=71.0708, parameterName=WaveDistance}
  n47 = Sine(None=Frac(None=Divide(A=@n45, B=@n46)))
  n48 = ScalarParameter() {defaultValue=5.2267, parameterName=WaveyWPO}
  n49 = Add(
   A=StaticSwitchParameter(True=@n17, False=Add(A=@n44, B=@n20)) {parameterName=IsBottomBlend}
   B=Multiply(A=Constant3Vector() {constant=<rgba 1 1 0 0>}, B=Multiply(A=@n47, B=@n48))
  )
  Add(A=@n31, B=@n49)
  Add()
  n50 = Transform(None=Constant3Vector() {constant=<rgba 1 0 0 0>}) {transformSourceType=TRANSFORMSOURCE_Local}
  n51 = Transform(None=Constant3Vector() {constant=<rgba 1 1 0 0>}) {transformSourceType=TRANSFORMSOURCE_Local}
  AppendVector(
   A=ComponentMask(None=MaterialFunctionCall(Vector 3=@n50) {materialFunction=/Engine/Functions/Engine_MaterialFunctions02/Utility/VectorLength}.V3 Length) {r=true}
   B=ComponentMask(None=MaterialFunctionCall(Vector 3=@n51) {materialFunction=&VectorLength}.V3 Length) {g=true}
  )
  ComponentMask(None=TextureCoordinate() {coordinateIndex=1}) {g=true}
  n52 = Transform(None=AppendVector(A=@n3, B=Constant())) {transformSourceType=TRANSFORMSOURCE_Local}
  ComponentMask(None=@n52) {g=true, r=true}
  Divide()
  n53 = WorldPosition()
  n54 = Constant()
  n55 = StaticSwitchParameter(True=@n57.BubblesUVDistortion, False=@n54) {defaultValue=true, parameterName=UseBubbles}
  n56 = MaterialFunctionCall!(
   UVs=Divide(A=ComponentMask(None=@n53.XYZ) {g=true, r=true}, B=ScalarParameter() {defaultValue=653.4476, parameterName=LavaSizeMaster})
   BubblesAdditive=StaticSwitchParameter(True=@n57.BubblesAdditive, False=@n54) {defaultValue=true, parameterName=UseBubbles}
   UVDistortion=@n55
  ) {materialFunction=&/MF_LavaMagma}
  n57 = MaterialFunctionCall!(In=@n56.Result, UVs=@n56.UVs Panning) {materialFunction=&/MF_LavaBubbles}
  n58 = StaticSwitchParameter(True=@n57.Result, False=@n56.Result) {defaultValue=true, parameterName=UseBubbles}
  n59 = VectorParameter() {parameterName=FallsColor, defaultValue=<rgba 2 0.25 0 1>}
  n60 = Power(
   Base=Add(A=TextureSample(UVs=@n40) {samplerType=SAMPLERTYPE_Masks, texture=/Game/Environments/Sets/Vulcano/Textures/Lava_M}.G, B=@n32.R)
   Exp=ScalarParameter() {defaultValue=1.2884, parameterName=powerRings}
  )
  n61 = ComponentMask(None=TextureCoordinate() {coordinateIndex=1}) {g=true}
  MaterialFunctionCall!(
   Base Material=@n58
   Top Material=MakeMaterialAttributes!(
    BaseColor
    Metallic
    Specular
    Roughness
    Anisotropy
    EmissiveColor=LinearInterpolate(B=@n59.RGB, Alpha=@n60)
    Opacity
    OpacityMask
    Normal=TextureSample(UVs=@n40) {samplerType=SAMPLERTYPE_Normal, texture=/Game/Environments/Sets/Vulcano/Textures/Lava_N}.RGB
    Tangent
    WorldPositionOffset=@n31
    SubsurfaceColor
    ClearCoat
    ClearCoatRoughness
    AmbientOcclusion
    Refraction
    CustomizedUVs_0
    CustomizedUVs_1
    CustomizedUVs_2
    CustomizedUVs_3
    CustomizedUVs_4
    CustomizedUVs_5
    CustomizedUVs_6
    CustomizedUVs_7
    PixelDepthOffset
    ShadingModel
    Displacement
   )
   Alpha=@n61
  ) {materialFunction=/Engine/Functions/MaterialLayerFunctions/MatLayerBlend_Standard}
  Multiply(A=@n52, B=ComponentMask(None=@n51) {g=true, r=true})
  n62 = TextureCoordinate()
  n63 = ComponentMask(None=@n62) {g=true}
  n64 = Multiply(A=@n63, B=OneMinus(None=@n63))
  n65 = ObjectPositionWS()
  n66 = Add(A=@n62, B=Frac(None=ComponentMask(None=@n65) {g=true, r=true}))
  n67 = AppendVector(A=Constant(), B=Multiply(A=Time(), B=ScalarParameter() {defaultValue=-0.4, parameterName=FallsSpeed1}))
  n68 = ScalarParameter() {defaultValue=1, parameterName=FallsTilingX}
  n69 = ScalarParameter() {defaultValue=2.4505, parameterName=FallsTilingZ}
  n70 = Frac(None=Divide(A=Add(A=@n66, B=@n67), B=AppendVector(A=@n68, B=@n69)))
  n71 = OneMinus(None=TextureSample(UVs=@n70) {samplerType=SAMPLERTYPE_Masks, texture=&Lava_M}.G)
  n72 = MaterialFunctionCall(
   Transition Phase=Multiply(A=OneMinus(None=@n64), B=ScalarParameter() {defaultValue=0.6605, parameterName=HeightblendFalls})
   Height Texture=@n71
  ) {materialFunction=&HeightLerp}
  n73 = ScalarParameter() {defaultValue=-1.689, parameterName=InnerEdge2}
  Multiply(A=OneMinus(None=@n72.Lerp Alpha No Contrast), B=@n73)
  n74 = VectorParameter() {parameterName=PileColor, defaultValue=<rgba 9 1.125 0 1>}
  n75 = Power(Base=Max(A=@n14, B=@n13.G), Exp=ScalarParameter() {defaultValue=1.6371, parameterName=powerRings2})
  n76 = MaterialFunctionCall(
   Base Material=@n58
   Top Material=MakeMaterialAttributes(
    EmissiveColor=LinearInterpolate(B=@n74.RGB, Alpha=@n75)
    Normal=TextureSample!(UVs=@n9, Tex, "DDX(UVs)"=@n11, "DDY(UVs)"=@n12, Apply View MipBias) {mipValueMode=TMVM_Derivative, samplerType=SAMPLERTYPE_Normal, texture=/Game/Environments/Sets/Vulcano/Textures/LavaPileUp_N}.RGB
   )
   Alpha=@n61
  ) {materialFunction=&MatLayerBlend_Standard}
  n77 = VectorParameter() {parameterName=FallsColor, defaultValue=<rgba 2 0.25 0 1>}
  n78 = Clamp(None=LinearInterpolate(
    A=ScalarParameter() {defaultValue=-6.5449, parameterName=GradienBottom}
    B=ScalarParameter() {defaultValue=3.4402, parameterName=GradienBottom2}
    Alpha=@n63
   )) {clampMode=CMODE_ClampMin}
  n79 = ComponentMask(None=@n62) {r=true}
  n80 = OneMinus(None=@n79)
  n81 = Add(
   A=LinearInterpolate(
    A=ScalarParameter() {defaultValue=-0.7775, parameterName=FalloffEdges}
    B=ScalarParameter() {defaultValue=2.7126, parameterName=FalloffEdges2}
    Alpha=Multiply(A=@n79, B=@n80)
   )
   B=@n71
  )
  n82 = ScalarParameter() {defaultValue=6.9247, parameterName=InnerEdge1}
  n83 = Add(
   A=Power(Base=Add(A=@n78, B=@n81), Exp=ScalarParameter() {defaultValue=1.1253, parameterName=power})
   B=Clamp(None=LinearInterpolate(A=@n82, B=@n73, Alpha=@n72.Lerp Alpha No Contrast)) {clampMode=CMODE_ClampMin}
  )
  n84 = MaterialFunctionCall(
   Base Material=@n58
   Top Material=MakeMaterialAttributes(
    EmissiveColor=LinearInterpolate(B=@n77.RGB, Alpha=@n83)
    OpacityMask=@n72.Alpha
    Normal=TextureSample!(UVs=@n70, Tex, Apply View MipBias) {samplerType=SAMPLERTYPE_Normal, texture=&Lava_N}.RGB
    WorldPositionOffset=@n31
   )
   Alpha=ComponentMask(None=TextureCoordinate() {coordinateIndex=3}) {g=true}
  ) {materialFunction=&MatLayerBlend_Standard}
  SetMaterialAttributes!(
   MaterialAttributes=GetMaterialAttributes(None=StaticSwitchParameter(True=@n76.Blended Material, False=@n84.Blended Material) {parameterName=IsBottomBlend}) {attributeGetTypes [#1]}.MaterialAttributes
   World Position Offset=@n49
  ) {attributeSetTypes [#1]}
  n85 = Multiply(A=@n13.G, B=@n32.R)
  n86 = Multiply(A=Normalize(VectorInput=@n2), B=ScalarParameter() {defaultValue=-0.5628, parameterName=Distortion2})
  n87 = Multiply(A=Multiply(A=ScalarParameter() {defaultValue=0.176, parameterName=Distortion}, B=Normalize(VectorInput=@n34)), B=@n32.R)
  StaticSwitchParameter(True=Add(A=Multiply(A=@n85, B=@n86), B=@n55), False=Add(A=@n55, B=Multiply(A=@n87, B=@n43))) {parameterName=IsBottomBlend}
  Subtract(
   A=@n1.XYZ
   B=TransformPosition(None=VectorParameter() {parameterName=LocationPileCenter, defaultValue=<rgba 0 39.1026 0 1>}.RGB) {transformType=TRANSFORMPOSSOURCE_World}
  )
  Transform(None=MaterialFunctionCall() {materialFunction=/Engine/Functions/Engine_MaterialFunctions02/WorldPositionOffset/ObjectScale}.Scale XYZ) {transformSourceType=TRANSFORMSOURCE_World, transformType=TRANSFORM_Local}
  Transform(None=@n4) {transformSourceType=TRANSFORMSOURCE_Local}
  VertexColor()
  VertexColor()
  VertexColor()
  VertexColor()
  VertexColor()
  VertexColor()
  VertexColor()
  WorldPosition()
 ]
}
