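# Translucent unlit material for an under-lava effect, drawing an animated orange glow around the actor's height.
# Vertex math re-anchors the mesh to the camera, shaping it from UVs, camera-space offsets and camera height relative to the actor.
# Two world-space sine waves ripple the vertices and the gradient lookup blending three colors, while depth below actor height drives opacity.
# An optional LimitAreaBox switch collapses vertices when the camera leaves an X/Y range.
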Material /Game/Environments/Sets/Vulcano/Materials/M_UnderLavaFX
graph {
 properties {blendMode=BLEND_Translucent, pixelDepthOffsetMode=PDOM_Legacy, shadingModel=MSM_Unlit, usedShadingModels=Unlit}
 nodes [
  n1 = ScalarParameter() {defaultValue=12, parameterName=Forward2}
  Divide(A=@n1)
  n2 = Frac(None=Multiply(A=Time(), B=ScalarParameter() {defaultValue=1.2334, parameterName=Speedanim}))
  n3 = WorldPosition()
  n4 = Add(A=@n2, B=Divide(A=ComponentMask(None=@n3.XYZ) {r=true}, B=ScalarParameter() {defaultValue=23.3664, parameterName=WaveLength1}))
  n5 = Add(A=@n2, B=Divide(A=ComponentMask(None=@n3.XYZ) {g=true}, B=ScalarParameter() {defaultValue=41.6767, parameterName=WaveLength2}))
  n6 = Add(
   A=Multiply(A=Sine(None=@n4), B=ScalarParameter() {defaultValue=0.9037, parameterName=Amplitude})
   B=Multiply(A=Sine(None=@n5), B=ScalarParameter() {defaultValue=1, parameterName=Amplitude2})
  )
  n7 = ComponentMask(None=ActorPositionWS()) {b=true}
  n8 = Subtract(A=@n7, B=ComponentMask(None=CameraPositionWS()) {b=true})
  n9 = Clamp(None=@n8) {clampMode=CMODE_ClampMax, maxDefault=0}
  n10 = TextureCoordinate()
  n11 = ComponentMask(None=@n10) {g=true}
  n12 = Multiply(A=Clamp(None=@n8) {clampMode=CMODE_ClampMin, maxDefault=0}, B=OneMinus(None=@n11))
  n13 = Transform(None=VectorParameter() {parameterName=Direction, defaultValue=<rgba 0 1 0 1>}.RGB) {transformSourceType=TRANSFORMSOURCE_Camera}
  n14 = Constant()
  n15 = Transform(None=Constant3Vector() {constant=<rgba 0 0 1 0>}) {transformSourceType=TRANSFORMSOURCE_Camera}
  n16 = Multiply(A=@n15, B=Constant3Vector() {constant=<rgba 1 1 0 0>})
  n17 = MaterialFunctionCall(Vector=@n16) {materialFunction=&SafeNormalize}
  n18 = ScalarParameter() {defaultValue=12, parameterName=PushForward2}
  n19 = Constant3Vector() {constant=<rgba 0 0 -1 0>}
  n20 = Clamp(None=OneMinus(None=DotProduct(A=@n15, B=@n19))) {minDefault=0}
  n21 = Multiply(A=Normalize(VectorInput=@n16), B=LinearInterpolate(A=Divide(A=@n18, B=@n20), B=ScalarParameter() {parameterName=PushForward}, Alpha=@n11))
  n22 = Constant()
  n23 = ScalarParameter() {defaultValue=10666.3252, parameterName=Up}
  n24 = ScalarParameter() {defaultValue=12, parameterName=Forward3}
  n25 = Multiply(
   A=Transform(None=MaterialFunctionCall(X=@n22, Y=@n23, Z=@n24) {materialFunction=&MakeFloat3}.Result) {transformSourceType=TRANSFORMSOURCE_Camera}
   B=OneMinus(None=@n11)
  )
  n26 = Multiply(A=Subtract(A=ComponentMask(None=@n10) {r=true}) {constB=0.5}) {constB=2}
  n27 = ScalarParameter() {defaultValue=149.776, parameterName=LeftRight}
  n28 = ScalarParameter() {defaultValue=-52.858, parameterName=Down}
  n29 = ScalarParameter() {parameterName=Forward}
  n30 = Transform(None=MaterialFunctionCall(X=Multiply(A=@n26, B=@n27), Y=Multiply(A=@n11, B=@n28), Z=LinearInterpolate(A=@n29, B=@n1, Alpha=@n11)) {materialFunction=&MakeFloat3}.Result) {transformSourceType=TRANSFORMSOURCE_Camera}
  n31 = ActorPositionWS()
  n32 = Subtract(A=@n31, B=WorldPosition().XYZ)
  n33 = Add(A=Subtract(A=CameraPositionWS(), B=@n31), B=@n32)
  n34 = Add(
   A=Multiply(A=Add(A=@n9, B=@n12), B=@n13)
   B=Add(A=If(A=@n14, B=@n17."Length==0", "A > B"=@n21, "A == B"=@n21, "A < B"=@n25), B=Add(A=@n30, B=@n33))
  )
  n35 = Subtract(A=ComponentMask(None=Add(A=@n3.XYZ, B=@n34)) {b=true}, B=@n7)
  n36 = TextureSample(UVs=Divide(A=Add(A=@n6, B=@n35), B=ScalarParameter() {defaultValue=69.6328, parameterName=GradientDistance})) {texture=/Game/Global/Textures/Default/Masks_and_Gradients/T_Vert_Gradient_Sections}
  n37 = ScalarParameter() {defaultValue=1.7495, parameterName=Power}
  n38 = Clamp(None=Divide(A=@n35, B=ScalarParameter() {defaultValue=-12.2384, parameterName=GradientDistance2}))
  n39 = Clamp(None=Divide(A=Abs(None=@n35), B=ScalarParameter() {defaultValue=128, parameterName=GradientDistance3}))
  LinearInterpolate(
   A=VectorParameter() {parameterName=Color3, defaultValue=<rgba 0.035 0.0022 0 1>}.RGB
   B=LinearInterpolate(
    A=VectorParameter() {parameterName=Color, defaultValue=<rgba 0.5781 0.0367 0 1>}.RGB
    B=VectorParameter() {parameterName=Color2, defaultValue=<rgba 10 1.3972 0 1>}.RGB
    Alpha=Power(Base=@n36.R, Exp=@n37)
   )
   Alpha=Multiply(A=@n38, B=OneMinus(None=@n39))
  )
  Multiply(A=@n32, B=Constant3Vector() {constant=<rgba 1 1 0 0>})
  Multiply(A=@n38, B=ScalarParameter() {defaultValue=32.6953, parameterName=Opacity})
  n40 = Subtract(A=ComponentMask(None=ObjectPositionWS()) {g=true}, B=ComponentMask(None=CameraPositionWS()) {g=true})
  n41 = Floor(None=Subtract(A=Abs(None=@n40), B=ScalarParameter() {defaultValue=917, parameterName=FadeDistanceY}))
  Saturate(None=@n41)
  n42 = Add(A=@n6, B=ScalarParameter() {defaultValue=3.9202, parameterName=VetexShaderWaveadd})
  n43 = Add(A=Multiply(A=ScalarParameter() {defaultValue=0.2, parameterName=VetexShaderWave}, B=Multiply(A=@n42, B=@n13)), B=@n34)
  n44 = ObjectPositionWS()
  n45 = WorldPosition()
  n46 = ObjectPositionWS()
  n47 = CameraPositionWS()
  n48 = Abs(None=Subtract(A=ComponentMask(None=@n46) {r=true}, B=ComponentMask(None=@n47) {r=true}))
  n49 = ScalarParameter() {defaultValue=917.2893, parameterName=FadeDistanceX}
  n50 = Add(A=Floor(None=Subtract(A=@n48, B=@n49)), B=@n41)
  StaticSwitchParameter(True=LinearInterpolate(A=@n43, B=Subtract(A=@n44, B=@n45.XYZ), Alpha=Saturate(None=@n50)), False=@n43) {parameterName=LimitAreaBox}
 ]
}
