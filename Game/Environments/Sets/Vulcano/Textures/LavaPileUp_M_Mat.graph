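# Material asset in the Vulcano environment set, named after the LavaPileUp_M mask texture, with the Default Lit shading model.
# The graph holds a single Texture Sample of LavaPileUp_M using the Masks sampler type.
# No connection from the sample to a material output appears, so it reads as an unfinished stub.
# Automatic setting of usage flags in the editor is turned off.
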Material /Game/Environments/Sets/Vulcano/Textures/LavaPileUp_M_Mat
graph {
 properties {bAutomaticallySetUsageInEditor=false, usedShadingModels=Default Lit}
 nodes [TextureSample() {samplerType=SAMPLERTYPE_Masks, texture=&/LavaPileUp_M}]
}
