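# Material function for the Vulcano set that blends a glowing, darkened lava-ground edge with an input material along a transition gradient.
# MF_LavaGround supplies the edge surface, switchable between world-aligned and UV-mapped.
# Its inverted height and the gradient (default UV1 green) drive a HeightLerp that masks two-color emissive, compensated for eye adaptation, and darkens base color.
# A second HeightLerp, using the Height input and inverted gradient, layers the input material on top.
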MaterialFunction /Game/Environments/Sets/Vulcano/Materials/MF_LavaEdgeGradient
graph {nodes [
  n1 = TextureCoordinate()
  n2 = ScalarParameter() {defaultValue=76.7467, parameterName=LavaEdgeTiling}
  n3 = MaterialFunctionCall!(
   UV=FunctionInput(Preview=Multiply(A=@n1, B=@n2)) {bUsePreviewValueAsDefault=true, inputName=UVs, inputType=FunctionInput_Vector2, sortPriority=3}
   WorldAlignedSize=ScalarParameter() {defaultValue=185.6007, parameterName=LavaGroundEdge_Size}
  ) {materialFunction=&/MF_LavaGround}
  n4 = GetMaterialAttributes(None=@n3.UVMapped) {attributeGetTypes [#1]}
  n5 = Transform(None=@n4.Normal)
  n6 = GetMaterialAttributes(None=StaticSwitchParameter(
    True=@n3.Result
    False=SetMaterialAttributes!(MaterialAttributes=@n4.MaterialAttributes, Normal=@n5) {attributeSetTypes [#1]}
   ) {parameterName=EdgeGradientuseWorldAligned}) {attributeGetTypes [#2]}
  n7 = VectorParameter() {parameterName=GradientColor1, defaultValue=<rgba 0 0 0 1>}
  n8 = VectorParameter() {parameterName=GradientColor2, defaultValue=<rgba 1 0.082 0 1>}
  n9 = OneMinus(None=@n3.Height)
  n10 = Power(Base=@n9, Exp=ScalarParameter() {defaultValue=3.1021, parameterName=EmissiveEdgeContrast})
  n11 = FunctionInput(Preview=ComponentMask(None=TextureCoordinate() {coordinateIndex=1}) {g=true}) {bUsePreviewValueAsDefault=true, inputName=TransitionGradient, inputType=FunctionInput_Scalar, sortPriority=4}
  n12 = MaterialFunctionCall(
   Transition Phase=Multiply(A=@n11, B=ScalarParameter() {defaultValue=0.658, parameterName=EmissiveEdgeHeightLerp})
   Height Texture=@n9
   Contrast=ScalarParameter() {defaultValue=1.1432, parameterName=ContrastEdgeEmissive}
  ) {materialFunction=&HeightLerp}
  n13 = EyeAdaptationInverse(
   LightValueInput=Multiply(A=LinearInterpolate(A=@n7.RGB, B=@n8.RGB, Alpha=@n10), B=@n12.Alpha)
   AlphaInput=ScalarParameter() {defaultValue=0.5, parameterName=EyeAdaptationRatio}
  )
  n14 = ScalarParameter() {defaultValue=1, parameterName=DarkenEdge1}
  n15 = ScalarParameter() {defaultValue=-2.615, parameterName=DarkenEdge2}
  n16 = Multiply(A=@n6.BaseColor, B=Clamp(None=LinearInterpolate(A=@n14, B=@n15, Alpha=@n12.Lerp Alpha No Contrast)))
  n17 = OneMinus(None=Multiply(A=@n11, B=ScalarParameter() {defaultValue=1.2151, parameterName=TransitionEdgeLerp}))
  n18 = FunctionInput(Preview=Constant() {r=1}) {bUsePreviewValueAsDefault=true, inputName=Height, inputType=FunctionInput_Scalar, sortPriority=1}
  n19 = ScalarParameter() {defaultValue=1.6258, parameterName=TransitionEdgeLerpContrast}
  FunctionOutput(None=MaterialFunctionCall!(
    Base Material=SetMaterialAttributes!(MaterialAttributes=@n6.MaterialAttributes, EmissiveColor=@n13.EyeAdaptationInverse, BaseColor=@n16) {attributeSetTypes [#3, #2]}
    Top Material=FunctionInput() {inputType=FunctionInput_MaterialAttributes}
    Alpha=MaterialFunctionCall(Transition Phase=@n17, Height Texture=@n18, Contrast=@n19) {materialFunction=&HeightLerp}.Alpha
   ) {materialFunction=/Engine/Functions/MaterialLayerFunctions/MatLayerBlend_Standard}.Blended Material)
  OneMinus()
 ]}
