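# Material function for the dark soil layer of the Vulcano lava environment, built from the Arid set's rocky desert sand textures.
# UVs default to world XY divided by a size parameter, giving planar tiling unless the input is overridden.
# Base color is the color map's blue channel scaled by a brightness parameter, producing near-black greyscale.
# The mask texture supplies scaled specular, roughness, and a separate Height output alongside the material attributes with normal map.
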MaterialFunction /Game/Environments/Sets/Vulcano/Materials/MF_Lava_Soil
graph {nodes [
  n1 = WorldPosition()
  n2 = FunctionInput(Preview=Divide(A=ComponentMask(None=@n1.XYZ) {g=true, r=true}, B=ScalarParameter() {defaultValue=512, parameterName=LavaSoil_Size})) {bUsePreviewValueAsDefault=true, inputName=UVs, inputType=FunctionInput_Vector2}
  n3 = TextureSample(UVs=@n2) {texture=/Game/Environments/Sets/Arid/Textures/T_desert_sand_rocky_C}
  n4 = ScalarParameter() {defaultValue=0.05, parameterName=LavaSoilBirghtness}
  n5 = ScalarParameter() {defaultValue=0.4286, parameterName=LavaSoilSpecular}
  n6 = TextureSample(UVs=@n2) {samplerType=SAMPLERTYPE_Masks, texture=/Game/Environments/Sets/Arid/Textures/T_desert_sand_rocky_S}
  FunctionOutput(None=MakeMaterialAttributes!(
    BaseColor=Multiply(A=@n3.B, B=@n4)
    Metallic
    Specular=Multiply(A=@n5, B=@n6.R)
    Roughness=@n6.B
    Anisotropy
    EmissiveColor
    Opacity
    OpacityMask
    Normal=TextureSample(UVs=@n2) {samplerType=SAMPLERTYPE_Normal, texture=/Game/Environments/Sets/Arid/Textures/T_desert_sand_rocky_N}.RGB
    Tangent
    WorldPositionOffset
    SubsurfaceColor
    ClearCoat
    ClearCoatRoughness
    AmbientOcclusion
    Refraction
    CustomizedUVs_0
    CustomizedUVs_1
    CustomizedUVs_2
    CustomizedUVs_3
    CustomizedUVs_4
    CustomizedUVs_5
    CustomizedUVs_6
    CustomizedUVs_7
    PixelDepthOffset
    ShadingModel
    Displacement
   ))
  FunctionOutput(None=@n6.G) {outputName=Height}
 ]}
